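# Niagara debris and smoke effect for the PistolCanary castle ground churn, throwing bricks, chunks, wall crumbles and smoke around several towers.
# Most emitters partition particles across system TowerPositions and TowerFloorSize arrays, using vector curves to trace square footprints and place particles per wall.
# Bricks orient to each wall, while chunk and low smoke spawns contract toward the origin over the loop.
# MeshSmoke scales dynamic material parameters from one Global Scale; a Preview emitter is disabled.
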NiagaraSystem /PistolCanary/Assets/FX/Debris/NS_PistolCanary_Castle_GroundChurn
dependencies {
 usedDataInterfaces [
  &NiagaraDataInterfaceArrayFloat
  &NiagaraDataInterfaceVectorCurve
  &NiagaraDataInterfaceCurve
  &NiagaraDataInterfaceColorCurve
  &NiagaraDataInterfaceArrayInt32
  &NiagaraDataInterfaceMeshRendererInfo
  &NiagaraDataInterfaceArrayFloat3
 ]
 usedDynamicInputs [&Vector2DFromFloat]
 usedModules [
  &self:SystemSpawnScript.SystemScriptSource.SystemScriptGraph.NiagaraNodeAssignment_4.SetVariables_#1
  &SystemState
  &self:SystemSpawnScript.SystemScriptSource.SystemScriptGraph.NiagaraNodeAssignment_2.SetVariables_#2
  &self:SystemSpawnScript.SystemScriptSource.SystemScriptGraph.NiagaraNodeAssignment_0.SetVariables_#3
 ]
}
emitters {
 Bricks {
  emitterUpdateScript [
   &EmitterState(Life Cycle Mode=::NewEnumerator1, Loop Behavior=::NewEnumerator1, Loop Delay=1.5, Loop Duration=1, UseLoopDelay=false)
   &SpawnBurst_Instantaneous(Spawn Count=25, Use Spawn Probability=false, UseLoopCountLimit=false)
   &SpawnBurst_Instantaneous(Spawn Count=25, Spawn Time=0.06, Use Spawn Probability=false, UseLoopCountLimit=false) {module=SpawnBurst_Instantaneous001}
   &SpawnRate(
    SpawnRate=&FloatFromCurve(
     FloatCurve=DataInterface() {propertyValues {Curve {keys [
         {arriveTangent=-0.7676, interpMode=RCIM_Cubic, leaveTangent=-0.7676, tangentMode=RCTM_User, time=0, value=1}
         {arriveTangent=-1.4079, interpMode=RCIM_Cubic, leaveTangent=-1.4079, tangentMode=RCTM_User, time=1, value=0}
        ]}}}
     CurveIndex=$Emitter.NormalizedLoopAge
     Scale Curve=500
    )
    Use Spawn Probability=false
   )
  ]
  particleSpawnScript [
   &PartitionParticles(Number Of Partitions=4, Particles Per Partition=4, Partition Mode:ENiagara_PartitionMode_v2=::NewEnumerator1)
   /SetVariables(Particles.ZPos:NiagaraFloat=&RandomRangeFloat(Minimum=0.2, Maximum=1)) {isSetParameters=true}
   /SetVariables(Particles.WallDirection:Vector3f=&RotateVectorByQuaternion(
     Vector To Rotate=<0 1 0>
     Quaternion=&MakeQuaternion(
      Quaternion From=0
      Axis=<0 0 1>
      Angle=&Multiply_FloatByInt(Float=0.25, Integer=$Particles.PartitionParticles.Index In Partition)
      From=?
      To=?
      Angle Type=::NewEnumerator1
      Normalize Vectors=false
     )
    )) {isSetParameters=true}
   &InitializeParticle(
    Lifetime Max=1.2
    Lifetime Mode=::NewEnumerator1
    Mesh Renderer Array Visibility Mode=::NewEnumerator1
    Mesh Renderer Info=DataInterface() {propertyValues {MeshRenderer=&self:bricks_0.NiagaraMeshRendererProperties_1}}
    Mesh Renderer Mesh Index:NiagaraInt32=&SelectIntFromArray(
     Array Sampling Mode=::NewEnumerator1
     Direct Array Index=?
     Int Selection Array=DataInterface() {propertyValues {IntData [0,0,0,1]}}
     Random Seed=0
    )
    Mesh Scale Max=<1.5 1.5 1.5>
    Mesh Scale Mode=::NewEnumerator2
    Position Offset=&SelectVectorFromArray(Direct Array Index=$Particles.PartitionParticles.Partition, Vector Selection Array=$System.TowerPositions)
    Position Offset Coordinate Space=::Local
    UsePositionOffset=true
    Write Material Random=true
   )
   /SetVariables(Particles.Position=&AddVectorToPosition(
     Position=$Particles.Position
     Vector=&Multiply_Vector(
      A=&VectorFromCurve(
       VectorCurve=DataInterface() {propertyValues {
         XCurve {keys [<at 0 1>,<at 0.25 -1>,<at 0.5 -1>,<at 0.75 1>]}
         YCurve {keys [<at 0.25 1>,<at 0.5 -1>,<at 0.75 -1>,<at 1 1>]}
         ZCurve {keys [<at 0 1>,<at 1 1>]}
        }}
       CurveIndex=&Add_Float(
        A=&RandomRangeFloat(Minimum=0, Maximum=0.25)
        B=&Multiply_FloatByInt(Float=0.25, Integer=$Particles.PartitionParticles.Index In Partition)
       )
       Scale Curve=&Add_Vector(
        A=&Multiply_VectorByFloat(
         Vector=<1 1 0>
         Float=&Lerp_Float(
          A=1
          B=&SelectFloatFromArray(
           Direct Array Index=$Particles.PartitionParticles.Partition
           Float Selection Array=DataInterface() {propertyValues {FloatData [0.85,0.8,0.9,0.7]}}
          )
          Alpha=$Particles.ZPos
         )
        )
        B=&MakeVector(X=0, Y=0, Z=$Particles.ZPos)
       )
      )
      B=&SelectVectorFromArray(Direct Array Index=$Particles.PartitionParticles.Partition, Vector Selection Array=$System.TowerFloorSize)
     )
    )) {isSetParameters=true}
   &AddVelocity(
    Module.Use Falloff Distance=false
    Override Seed=false
    Velocity Mode=::NewEnumerator1
    Velocity Speed=&RandomRangeFloat(Minimum=50, Maximum=100)
   ) {module=AddVelocity001}
   &AddVelocity(Velocity=$Particles.WallDirection, Velocity Speed Scale=&RandomRangeFloat(Minimum=50, Maximum=100))
   &InitialMeshOrientation(
    Mesh Axis To Orient=<0 1 0>
    Mesh Orientation Mode=::NewEnumerator2
    Orientation Vector=$Particles.WallDirection
    Rotation=&MakeVector(
     X=0
     Y=0
     Z=&UniformAOrBFloat(A=0, B=0.5, Evaluation Type=::NewEnumerator0, Recalculate Random Each Loop=true, Override Seed=false)
    )
    Use Rotation Vector=true
   )
  ]
  particleUpdateScript [
   &ParticleState()
   &CameraOffset(Camera Offset Amount=&FloatFromCurve(
     FloatCurve=DataInterface() {propertyValues {Curve {
        defaultValue=1
        keys [
         {arriveTangent=9.7662, interpMode=RCIM_Cubic, leaveTangent=9.7662, tangentMode=RCTM_User, time=0, value=-0.15}
         {arriveTangent=0.15, interpMode=RCIM_Cubic, leaveTangent=0.15, time=0.25, value=1}
         <at 1 0>
        ]
       }}}
     Scale Curve=200
    ))
   &ScaleMeshSize(Scale Factor=&VectorFromFloat(Value=&FloatFromCurve(FloatCurve=DataInterface() {propertyValues {Curve {keys [
          {arriveTangent=38.9432, interpMode=RCIM_Cubic, leaveTangent=38.9432, tangentMode=RCTM_User, time=0, value=0}
          {arriveTangentWeight=0.0267, interpMode=RCIM_Cubic, leaveTangentWeight=0.4002, tangentMode=RCTM_User, time=0.05, value=1}
          {arriveTangentWeight=0.363, interpMode=RCIM_Cubic, leaveTangentWeight=0.0968, tangentMode=RCTM_User, time=0.8, value=1}
          {arriveTangent=-7.4853, tangentMode=RCTM_User, time=1, value=0}
         ]}}})))
   &UpdateMeshOrientation(
    Rotation Coordinate Space=::NewEnumerator1
    Rotation Rate=&FloatFromCurve(
     FloatCurve=DataInterface() {propertyValues {Curve {
        defaultValue=1
        keys [
         {arriveTangent=5.4367, interpMode=RCIM_Cubic, leaveTangent=5.4367, tangentMode=RCTM_User, time=0, value=0.5}
         {arriveTangent=-0.1, interpMode=RCIM_Cubic, leaveTangent=-0.1, time=0.25, value=1}
         {arriveTangent=-1.095, interpMode=RCIM_Cubic, leaveTangent=-1.095, tangentMode=RCTM_User, time=1, value=0.4}
        ]
       }}}
     CurveIndex=$Particles.Age
     Scale Curve=&RandomRangeFloat(Minimum=-1, Maximum=-0.5)
    )
    Rotation Vector=&Add_Vector(
     A=/Niagara/DynamicInputs/Vectors/CrossProduct(Vector A:Vector3f=$Particles.WallDirection, Vector B:Vector3f=<0 0 1>, Normalize Incoming Vectors:NiagaraBool=true)
     B=&RandomVector(Vector Scale=&RandomRangeFloat(Minimum=-0.25, Maximum=0.25))
    )
   )
   &GravityForce(Gravity=&RandomRangeVector(Minimum=<0 0 -1200>, Maximum=<0 0 -960>))
   &SolveForcesAndVelocity(Clamp Velocity=false, Limit Acceleration=false)
  ]
  rendererClasses [&NiagaraMeshRendererProperties]
  renderers [{rendererClass=&NiagaraMeshRendererProperties, rendererIndex=0}]
 }
 DebrisChunks {
  emitterUpdateScript [
   &EmitterState()
   &SpawnRate(SpawnRate=30, Use Spawn Probability=false)
   &SpawnBurst_Instantaneous(Spawn Count=12, Spawn Time=0.12, Use Spawn Probability=false, UseLoopCountLimit=false)
   &SpawnBurst_Instantaneous(Spawn Count=12, Spawn Time=0.2, Use Spawn Probability=false, UseLoopCountLimit=false) {module=SpawnBurst_Instantaneous001}
  ]
  particleSpawnScript [
   &PartitionParticles(Number Of Partitions=4)
   &InitializeParticle(
    Color Maximum=<rgba 0.1384 0.1221 0.1119 1>
    Color Minimum=<rgba 0.0908 0.0908 0.0908 1>
    Color Mode=::NewEnumerator2
    Lifetime Min=1.5
    Lifetime Mode=::NewEnumerator1
    Mesh Scale Max=<6 12 12>
    Mesh Scale Min=<3.5 6 6>
    Mesh Scale Mode=::NewEnumerator2
    Position Offset=&SelectVectorFromArray(Direct Array Index=$Particles.PartitionParticles.Partition, Vector Selection Array=$System.TowerPositions)
    Position Offset Coordinate Space=::Local
    UsePositionOffset=true
    Write Material Random=true
   )
   /SetVariables(Particles.Position=&AddVectorToPosition(
     Position=$Particles.Position
     Vector=&Multiply_Vector(
      A=&VectorFromCurve(
       VectorCurve=DataInterface() {propertyValues {
         XCurve {keys [<at 0 1>,<at 0.25 -1>,<at 0.5 -1>,<at 0.75 1>]}
         YCurve {keys [<at 0.25 1>,<at 0.5 -1>,<at 0.75 -1>,<at 1 1>]}
         ZCurve {keys [<at 0 0>,<at 1 0>]}
        }}
       CurveIndex=&RandomRangeFloat(Minimum=0, Maximum=1)
       Scale Curve=<0.8 0.8 1>
      )
      B=&SelectVectorFromArray(Direct Array Index=$Particles.PartitionParticles.Partition, Vector Selection Array=$System.TowerFloorSize)
     )
    )) {isSetParameters=true, module=SetVariables_#4}
   /SetVariables(Particles.Position=&Multiply_VectorByFloat(
     Vector=Linked() {linkedVariable {name=Particles.Position, type=NiagaraPosition}}
     Float=&FloatFromCurve(
      FloatCurve=DataInterface() {propertyValues {Curve {keys [<at 0.2 1 RCIM_Cubic>, {arriveTangent=-0.674, interpMode=RCIM_Cubic, leaveTangent=-0.674, tangentMode=RCTM_User, time=1, value=0.6}]}}}
      CurveIndex=$Emitter.NormalizedLoopAge
     )
    )) {isSetParameters=true}
   &AddVelocity(
    Cone Axis=<0 0 1>
    Inner Cone Angle=15
    Module.UseConeFalloff=false
    Module.UseVelDistribution=false
    Override Seed=false
    Velocity Mode=::NewEnumerator2
    Velocity Speed=&RandomRangeFloat(Minimum=500, Maximum=800)
   )
   &InitialMeshOrientation(
    Mesh Orientation Mode=::NewEnumerator2
    Orientation Vector=$Particles.Velocity
    Rotation=&RandomRangeVector(Minimum=<0 0 0>, Maximum=<1 0 0>)
    Use Rotation Vector=true
   )
  ]
  particleUpdateScript [
   &ParticleState()
   &CameraOffset(Camera Offset Amount=&FloatFromCurve(
     FloatCurve=DataInterface() {propertyValues {Curve {
        defaultValue=1
        keys [
         {arriveTangent=5.7413, interpMode=RCIM_Cubic, leaveTangent=5.7413, tangentMode=RCTM_User, time=0, value=-0.1}
         {tangentMode=RCTM_User, time=0.5, value=1}
        ]
       }}}
     Scale Curve=150
    ))
   &ScaleMeshSize(Scale Factor=&VectorFromCurve(VectorCurve=DataInterface() {propertyValues {
       XCurve {keys [
         {arriveTangent=1.4676, interpMode=RCIM_Cubic, leaveTangent=1.4676, tangentMode=RCTM_User, time=0, value=0.1}
         {arriveTangent=0.2071, tangentMode=RCTM_User, time=1, value=1}
        ]}
       YCurve {keys [<at 0 0.25>,<at 1 1>]}
       ZCurve {keys [<at 0 0.25>,<at 1 1>]}
      }}))
   &DynamicMaterialParameters(
    Index 0 Param 1=&FloatFromCurve(FloatCurve=DataInterface() {propertyValues {Curve {
        defaultValue=1
        keys [
         {arriveTangent=32.5365, interpMode=RCIM_Cubic, leaveTangent=32.5365, tangentMode=RCTM_User, time=0, value=0}
         {arriveTangentWeight=0.1491, interpMode=RCIM_Cubic, leaveTangentWeight=0.2236, tangentMode=RCTM_User, time=0.05, value=1}
         {interpMode=RCIM_Cubic, tangentMode=RCTM_User, time=0.7, value=1}
         {arriveTangent=-4.1553, tangentMode=RCTM_User, time=1, value=0}
        ]
       }}})
    Index 0 Param 2=&RandomRangeFloat(Minimum=0.7, Maximum=2)
    Index 0 Param 3=&FloatFromCurve(
     FloatCurve=DataInterface() {propertyValues {Curve {keys [
         {arriveTangent=-2.3575, interpMode=RCIM_Cubic, leaveTangent=-2.3575, tangentMode=RCTM_User, time=0, value=1}
         {arriveTangent=-0.3909, tangentMode=RCTM_User, time=1, value=0}
        ]}}}
     Scale Curve=&RandomRangeFloat(Minimum=6, Maximum=12)
    )
    Module.Index0Float0Write=true
    Module.Index0Float1Write=true
    Module.Index0Float2Write=true
    Module.Index0Float3Write=false
   )
   &GravityForce(Gravity=&RandomRangeVector(Minimum=<0 0 -1200>, Maximum=<0 0 -960>))
   &SolveForcesAndVelocity(Clamp Velocity=false, Limit Acceleration=false)
  ]
  rendererClasses [&NiagaraMeshRendererProperties]
  renderers [{rendererClass=&NiagaraMeshRendererProperties, rendererIndex=0}]
 }
 LingerSmoke {
  emitterUpdateScript [
   &EmitterState()
   &SpawnRate(SpawnRate=20, Use Spawn Probability=false)
  ]
  particleSpawnScript [
   &InitializeParticle(
    Color=<rgba 0.02 0.019 0.0181 1.5>
    Lifetime=&RandomRangeFloat(Minimum=4, Maximum=5)
    Position Offset=&MakeVector(
     X=0
     Y=0
     Z=&Lerp_Float(
      A=500
      B=1000
      Alpha=&FloatFromCurve(
       FloatCurve=DataInterface() {propertyValues {Curve {defaultValue=1, keys [<at 0 0>, {arriveTangent=0.8379, tangentMode=RCTM_User, time=3, value=1}]}}}
       CurveIndex=$Emitter.Age
      )
     )
    )
    Sprite Rotation Mode=::NewEnumerator2
    Sprite Size Mode=::NewEnumerator3
    Uniform Sprite Size=&RandomRangeFloat(Minimum=2500, Maximum=3800)
    UsePositionOffset=true
    Write Material Random=true
   )
   &ShapeLocation(
    Cylinder Height=1000
    Cylinder Radius=2000
    MaskSpawnGroup=false
    Override Seed=false
    Shape Primitive=::NewEnumerator1
    Surface Only=true
    Surface Only Band Thickness=0.5
    Use Endcaps In Surface Only Mode=false
    UseOwnerScale=true
   )
   /SetVariables(Particles.InitSpriteRotRate=&UniformRangedFloat(Minimum=-3, Maximum=3)) {isSetParameters=true}
   /SetVariables(Particles.RandomFloat=&RandomRangeFloat(Minimum=0, Maximum=1)) {isSetParameters=true}
   /SetVariables(Particles.SubUVRandomStart:NiagaraInt32=&RandomRangeInt(Minimum=0, Maximum=44)) {isSetParameters=true}
   &AddVelocity(
    Module.Use Falloff Distance=false
    Origin Offset=<0 0 -2000>
    Override Seed=false
    Velocity Mode=::NewEnumerator1
    Velocity Speed=&RandomRangeFloat(Minimum=600, Maximum=900)
   )
  ]
  particleUpdateScript [
   &ParticleState()
   &WindForce(Turbulence Mode:ENiagaraWindTurbulenceMode=::NewEnumerator0, Wind Speed=<0 0 1>, Wind Speed Scale=500)
   &Drag(Drag=0.5, Use Linear Drag=true, Use Rotational Drag=false)
   &SolveForcesAndVelocity(Clamp Velocity=false, Limit Acceleration=false)
   &ScaleColor(
    Scale Alpha=&FloatFromCurve(FloatCurve=DataInterface() {propertyValues {Curve {keys [
         {arriveTangent=43.8033, interpMode=RCIM_Cubic, time=0, value=0}
         <at 0.25 1 RCIM_Cubic>
         {arriveTangent=-2.8482, interpMode=RCIM_Cubic, leaveTangent=-2.8482, tangentMode=RCTM_User, time=1, value=0}
        ]}}})
    ScaleA=true
    ScaleRGB=false
   )
   &SpriteRotationRate(Rotation Rate=&Multiply_Float(
     A=$Particles.InitSpriteRotRate
     B=&FloatFromCurve(
      FloatCurve=DataInterface() {propertyValues {Curve {keys [
          {arriveTangent=-2.157, interpMode=RCIM_Cubic, leaveTangent=-2.157, tangentMode=RCTM_User, time=-0.0046, value=1.0113}
          {interpMode=RCIM_Cubic, leaveTangent=-0.2496, time=1, value=0.16}
         ]}}}
      Scale Curve=20
     )
    ))
   &ScaleSpriteSize(Uniform Curve Sprite Scale=DataInterface() {propertyValues {Curve {keys [
        {arriveTangent=1.5213, interpMode=RCIM_Cubic, leaveTangent=1.5213, tangentMode=RCTM_User, time=0, value=0.4}
        {arriveTangent=0.3088, tangentMode=RCTM_User, time=1, value=1}
       ]}}})
   &CameraOffset(Camera Offset Amount=/Niagara/DynamicInputs/Math/CalculateParticleRadius(
     Radius Calculation Type:ENiagaraCollisionRadiusOptions=::NewEnumerator0
     Particle Radius:NiagaraFloat=1
     Mesh Dimensions:Vector3f=<1 1 1>
     Method for Calculating Particle Radius:ENiagaraCalculateRadiusOptions=::NewEnumerator2
    ))
   /SetVariables(Particles.SubImageIndex=&Lerp_Float(A=&RandomRangeFloat(Minimum=4, Maximum=12), B=63, Alpha=$Particles.NormalizedAge)) {isSetParameters=true}
  ]
  rendererClasses [&NiagaraSpriteRendererProperties]
  renderers [{rendererClass=&NiagaraSpriteRendererProperties}]
 }
 LowUnlitSmoke {
  emitterUpdateScript [
   &EmitterState()
   &SpawnRate(
    SpawnRate=&FloatFromCurve(
     FloatCurve=DataInterface() {propertyValues {Curve {keys [<at 0.4 1 RCIM_Cubic>, {arriveTangent=-1.9331, tangentMode=RCTM_User, time=1, value=0.25}]}}}
     CurveIndex=$Emitter.NormalizedLoopAge
     Scale Curve=100
    )
    Use Spawn Probability=false
   )
  ]
  particleSpawnScript [
   &PartitionParticles(
    Normalized Index In Partition Includes One=true
    Number Of Partitions=4
    Particles Per Partition=4
    Partition Mode:ENiagara_PartitionMode_v2=::NewEnumerator1
   )
   &InitializeParticle(
    Color Maximum=<rgba 0.05 0.027 0.027 1.5>
    Color Minimum=<rgba 0.04 0.0256 0.0256 1.5>
    Color Mode=::NewEnumerator2
    Lifetime=&RandomRangeFloat(Minimum=1, Maximum=1.5)
    Mesh Scale=&RandomRangeVector(Minimum=<2.5 2 2>, Maximum=<3 3.5 3.5>)
    Mesh Scale Mode=::NewEnumerator4
    Position Offset=&SelectVectorFromArray(Direct Array Index=$Particles.PartitionParticles.Partition, Vector Selection Array=$System.TowerPositions)
    Position Offset Coordinate Space=::Local
    Sprite Rotation Mode=::NewEnumerator2
    Sprite Size Mode=::NewEnumerator1
    Uniform Sprite Size Max=2500
    Uniform Sprite Size Min=1700
    UsePositionOffset=true
    Write Material Random=true
   )
   /SetVariables(Particles.Position=&AddVectorToPosition(
     Position=$Particles.Position
     Vector=&Multiply_Vector(
      A=&VectorFromCurve(
       VectorCurve=DataInterface() {propertyValues {
         XCurve {keys [<at 0 1>,<at 0.25 -1>,<at 0.5 -1>,<at 0.75 1>]}
         YCurve {keys [<at 0.25 1>,<at 0.5 -1>,<at 0.75 -1>,<at 1 1>]}
         ZCurve {keys [<at 0 0>,<at 1 0>]}
        }}
       CurveIndex=&Add_Float(
        A=&RandomRangeFloat(Minimum=0, Maximum=0.25)
        B=&Multiply_FloatByInt(Float=0.25, Integer=$Particles.PartitionParticles.Index In Partition)
       )
      )
      B=&SelectVectorFromArray(Direct Array Index=$Particles.PartitionParticles.Partition, Vector Selection Array=$System.TowerFloorSize)
     )
    )) {isSetParameters=true}
   /SetVariables(Particles.Position=&Multiply_VectorByFloat(
     Vector=Linked() {linkedVariable {name=Particles.Position, type=NiagaraPosition}}
     Float=&FloatFromCurve(
      FloatCurve=DataInterface() {propertyValues {Curve {keys [<at 0.2 1 RCIM_Cubic>, {arriveTangent=-0.674, interpMode=RCIM_Cubic, leaveTangent=-0.674, tangentMode=RCTM_User, time=1, value=0.6}]}}}
      CurveIndex=$Emitter.NormalizedLoopAge
     )
    )) {isSetParameters=true}
   &AddVelocity(Velocity=<0 0 1>, Velocity Speed Scale=&RandomRangeFloat(Minimum=500, Maximum=1000))
   &CameraOffset(Camera Offset Amount=300)
  ]
  particleUpdateScript [
   &ParticleState()
   &ScaleSpriteSize(Uniform Curve Sprite Scale=DataInterface() {propertyValues {Curve {
       defaultValue=1
       keys [
        {arriveTangent=1.221, interpMode=RCIM_Cubic, leaveTangent=1.221, tangentMode=RCTM_User, time=0, value=0.4}
        {interpMode=RCIM_Cubic, tangentMode=RCTM_User, time=1, value=1}
       ]
      }}})
   &Drag(Drag=&RandomRangeFloat(Minimum=1, Maximum=1.5), Use Linear Drag=true, Use Rotational Drag=false)
   &ScaleColor(
    Scale Alpha=&FloatFromCurve(FloatCurve=DataInterface() {propertyValues {Curve {
        defaultValue=1
        keys [
         {arriveTangent=8.3954, interpMode=RCIM_Cubic, leaveTangent=8.3954, tangentMode=RCTM_User, time=0, value=0}
         {
          arriveTangent=-0.5026
          arriveTangentWeight=0.1491
          interpMode=RCIM_Cubic
          leaveTangent=-0.5026
          leaveTangentWeight=0.2236
          tangentMode=RCTM_User
          time=0.2
          value=1.25
         }
         {arriveTangentWeight=0.1, interpMode=RCIM_Cubic, leaveTangentWeight=0.1667, tangentMode=RCTM_User, time=0.75, value=1}
         {arriveTangent=-6.06, tangentMode=RCTM_User, time=1, value=0}
        ]
       }}})
    ScaleA=true
    ScaleRGB=false
   )
   &SolveForcesAndVelocity(Clamp Velocity=false, Limit Acceleration=false)
   &DynamicMaterialParameters(
    Index 0 Param 1=&Add_Float(
     A=$Particles.MaterialRandom
     B=&FloatFromCurve(FloatCurve=DataInterface() {propertyValues {Curve {keys [
          {arriveTangent=2.7533, interpMode=RCIM_Cubic, leaveTangent=1.8958, tangentMode=RCTM_User, time=0, value=1}
          {arriveTangent=0.4122, interpMode=RCIM_Cubic, leaveTangent=0.4122, tangentMode=RCTM_User, time=0.321, value=1.3453}
          {arriveTangent=0.1513, interpMode=RCIM_Cubic, leaveTangent=0.1513, tangentMode=RCTM_User, time=1, value=1.5}
         ]}}})
    )
    Index 0 Param 3=&FloatFromCurve(
     FloatCurve=DataInterface() {propertyValues {Curve {keys [{arriveTangent=-11.209, interpMode=RCIM_Cubic, leaveTangent=-11.209, tangentMode=RCTM_User, time=0, value=1}, <at 0.2 0.05>, <at 0.5 0>]}}}
     Scale Curve=0.01
    )
    Index 0 Param 4=&FloatFromCurve(FloatCurve=DataInterface() {propertyValues {Curve {keys [<at 0.5 1 RCIM_Cubic>,<at 1 0>]}}})
    Index 2 Param 1=&FloatFromCurve(FloatCurve=DataInterface() {propertyValues {Curve {keys [
         {interpMode=RCIM_Cubic, leaveTangent=4.2485, tangentMode=RCTM_Break, time=0, value=0}
         {arriveTangent=1.0548, interpMode=RCIM_Cubic, leaveTangent=1.0548, tangentMode=RCTM_User, time=0.17, value=0.5715}
         {arriveTangent=0.4121, tangentMode=RCTM_User, time=1, value=1}
        ]}}})
    Index 2 Param 2=&FloatFromCurve(
     FloatCurve=DataInterface() {propertyValues {Curve {keys [
         {arriveTangent=0.4242, interpMode=RCIM_Cubic, leaveTangent=0.4242, tangentMode=RCTM_User, time=0, value=0.1}
         {arriveTangent=0.1815, interpMode=RCIM_Cubic, leaveTangent=0.1815, tangentMode=RCTM_User, time=1, value=0.4}
        ]}}}
     Scale Curve=&RandomRangeFloat(Minimum=0.5, Maximum=1)
    )
    Index 2 Param 3=&FloatFromCurve(
     FloatCurve=DataInterface() {propertyValues {Curve {keys [
         {interpMode=RCIM_Cubic, leaveTangent=1.2101, tangentMode=RCTM_Break, time=0, value=0.25}
         {arriveTangent=0.6182, interpMode=RCIM_Cubic, leaveTangent=11.1233, tangentMode=RCTM_Break, time=1, value=1}
        ]}}}
     Scale Curve=3
    )
    Index 2 Param 4=&FloatFromCurve(
     FloatCurve=DataInterface() {propertyValues {Curve {keys [
         {arriveTangent=-0.5563, interpMode=RCIM_Cubic, leaveTangent=-0.5563, tangentMode=RCTM_User, time=0, value=1}
         {interpMode=RCIM_Cubic, leaveTangent=-1.2956, tangentMode=RCTM_Break, time=1, value=0.75}
        ]}}}
     Scale Curve=2
    )
    Module.Index0Float0Write=true
    Module.Index0Float1Write=false
    Module.Index0Float2Write=true
    Module.Index0Float3Write=true
    Module.Index2Float0Write=true
    Module.Index2Float1Write=true
    Module.Index2Float2Write=true
    Module.Index2Float3Write=true
    Write Parameter Index 2=::NewEnumerator1
   )
  ]
  rendererClasses [&NiagaraSpriteRendererProperties]
  renderers [{rendererClass=&NiagaraSpriteRendererProperties}]
 }
 MeshSmoke {
  emitterSpawnScript [
   /SetVariables(Emitter.Global Color:LinearColor=<rgba 0.3303 0.0509 0.4 1>, Emitter.Global Scale=3) {isSetParameters=true}
   /SetVariables(
    Emitter.Color2:LinearColor=&MultiplyLinearColor(Linear Color=$Emitter.Global Color, Scale Factor=<1 1 1 1>)
    Emitter.Depth Fade:NiagaraFloat=&Multiply_Float(A=50, B=$Emitter.Global Scale)
    Emitter.Dissolve Noise Power=1.1
    Emitter.Dissolve Noise Scale:LinearColor=&MultiplyLinearColor(
     Linear Color=/Niagara/DynamicInputs/TypeConversions/MakeLinearColorFromFloat(Value:NiagaraFloat=&FloatFromCurve(FloatCurve=DataInterface() {propertyValues {Curve {keys [<at 0 2>,<at 1 1>,<at 2 0.5>]}}}, CurveIndex=$Emitter.Global Scale))
     Scale Factor=<0.0008 0.0008 0.0009 1>
    )
    Emitter.Near Camera Fade:NiagaraFloat=&Multiply_Float(A=100, B=$Emitter.Global Scale)
   ) {isSetParameters=true}
   /SetVariables(Emitter.Dissolve Scale Adjsutment:NiagaraFloat=&FloatFromCurve(
     FloatCurve=DataInterface() {propertyValues {Curve {keys [<at 0 3>,<at 1 1>,<at 5 0.5>,<at 20 0.4>,<at 100 0.4>]}}}
     CurveIndex=$Emitter.Global Scale
     Scale Curve=$Emitter.Global Scale
    )) {isSetParameters=true}
  ]
  emitterUpdateScript [
   &EmitterState()
   &SpawnRate(
    SpawnRate=&FloatFromCurve(
     FloatCurve=DataInterface() {propertyValues {Curve {keys [<at 0.4 1 RCIM_Cubic>, {arriveTangent=-1.9331, tangentMode=RCTM_User, time=1, value=0.25}]}}}
     CurveIndex=$Emitter.NormalizedLoopAge
     Scale Curve=100
    )
    Use Spawn Probability=false
   )
  ]
  particleSpawnScript [
   &PartitionParticles(
    Normalized Index In Partition Includes One=true
    Number Of Partitions=4
    Particles Per Partition=4
    Partition Mode:ENiagara_PartitionMode_v2=::NewEnumerator1
   )
   &InitializeParticle(
    Color=<rgba 0.0935 0.0881 0.12 1.25>
    Lifetime=&RandomRangeFloat(Minimum=1, Maximum=1.5)
    Mesh Scale=&RandomRangeVector(Minimum=<2.5 2 2>, Maximum=<3 3.5 3.5>)
    Mesh Scale Mode=::NewEnumerator4
    Position Offset=&SelectVectorFromArray(Direct Array Index=$Particles.PartitionParticles.Partition, Vector Selection Array=$System.TowerPositions)
    Position Offset Coordinate Space=::Local
    UsePositionOffset=true
    Write Material Random=true
   )
   /SetVariables(Particles.Position=&AddVectorToPosition(
     Position=$Particles.Position
     Vector=&Multiply_Vector(
      A=&VectorFromCurve(
       VectorCurve=DataInterface() {propertyValues {
         XCurve {keys [<at 0 1>,<at 0.25 -1>,<at 0.5 -1>,<at 0.75 1>]}
         YCurve {keys [<at 0.25 1>,<at 0.5 -1>,<at 0.75 -1>,<at 1 1>]}
         ZCurve {keys [<at 0 0>,<at 1 0>]}
        }}
       CurveIndex=&Add_Float(
        A=&RandomRangeFloat(Minimum=0, Maximum=0.25)
        B=&Multiply_FloatByInt(Float=0.25, Integer=$Particles.PartitionParticles.Index In Partition)
       )
      )
      B=&SelectVectorFromArray(Direct Array Index=$Particles.PartitionParticles.Partition, Vector Selection Array=$System.TowerFloorSize)
     )
    )) {isSetParameters=true}
   /SetVariables(Particles.Position=&Multiply_VectorByFloat(
     Vector=Linked() {linkedVariable {name=Particles.Position, type=NiagaraPosition}}
     Float=&FloatFromCurve(
      FloatCurve=DataInterface() {propertyValues {Curve {keys [<at 0.2 1 RCIM_Cubic>, {arriveTangent=-0.674, interpMode=RCIM_Cubic, leaveTangent=-0.674, tangentMode=RCTM_User, time=1, value=0.6}]}}}
      CurveIndex=$Emitter.NormalizedLoopAge
     )
    )) {isSetParameters=true}
   &AddVelocity(Velocity=<0 0 1>, Velocity Speed Scale=&RandomRangeFloat(Minimum=500, Maximum=1000))
   &CameraOffset(Camera Offset Amount=300)
   /SetVariables(Particles.Scale=&Multiply_VectorByFloat(Vector=$Particles.Scale, Float=$Emitter.Global Scale)) {isSetParameters=true}
  ]
  particleUpdateScript [
   &ParticleState()
   &UpdateMeshOrientation(Rotation Rate=&RandomRangeFloat(Minimum=-0.2, Maximum=0.2))
   &Drag(Drag=&RandomRangeFloat(Minimum=1, Maximum=1.5), Use Linear Drag=true, Use Rotational Drag=false)
   &DynamicMaterialParameters(
    Index 0 Param 1=&Multiply_Float(A=-100, B=$Emitter.Global Scale)
    Index 0 Param 2=&FloatFromCurve(FloatCurve=DataInterface() {propertyValues {Curve {keys [
         {tangentMode=RCTM_User, time=0, value=1}
         {arriveTangent=-0.0405, interpMode=RCIM_Cubic, leaveTangent=-0.0405, tangentMode=RCTM_User, time=0.4, value=1}
         {arriveTangent=-1.6592, interpMode=RCIM_Cubic, leaveTangent=-1.6592, tangentMode=RCTM_User, time=1, value=0.4}
        ]}}})
    Index 0 Param 3=&FloatFromCurve(
     FloatCurve=DataInterface() {propertyValues {Curve {keys [
         {arriveTangent=1.4725, interpMode=RCIM_Cubic, leaveTangent=1.4725, tangentMode=RCTM_User, time=0, value=0}
         {arriveTangent=0.4975, tangentMode=RCTM_User, time=1, value=1}
        ]}}}
     Scale Curve=&Multiply_Float(
      A=&RandomRangeFloat(Minimum=100, Maximum=150)
      B=&Multiply_Float(
       A=$Emitter.Global Scale
       B=&FloatFromCurve(FloatCurve=DataInterface() {propertyValues {Curve {keys [<at 0 1>,<at 100 0.9>,<at 300 0>]}}}, CurveIndex=$Emitter.Global Scale)
      )
     )
    )
    Index 0 Param 4=&FloatFromCurve(
     FloatCurve=DataInterface() {propertyValues {Curve {keys [
         {arriveTangent=0.1175, interpMode=RCIM_Cubic, leaveTangent=1.7315, tangentMode=RCTM_Break, time=0.1, value=-3}
         {arriveTangent=35.5342, interpMode=RCIM_Cubic, leaveTangent=14.4841, tangentMode=RCTM_Break, time=1, value=9}
        ]}}}
     Scale Curve=0.2
    )
    Index 1 Param 1=&FloatFromCurve(
     FloatCurve=DataInterface() {propertyValues {Curve {keys [
         {arriveTangent=-59.504, interpMode=RCIM_Cubic, leaveTangent=-59.504, tangentMode=RCTM_User, time=0, value=5}
         {arriveTangent=2.8571, interpMode=RCIM_Cubic, leaveTangent=2.8571, time=0.15, value=0.7}
         {interpMode=RCIM_Cubic, leaveTangent=-0.2867, time=0.7, value=7}
        ]}}}
     Scale Curve=3
    )
    Index 1 Param 2=&Divide_Float(
     A=&RandomRangeFloat(Minimum=0.016, Maximum=0.018)
     B=&Multiply_Float(
      A=$Emitter.Global Scale
      B=&FloatFromCurve(
       FloatCurve=DataInterface() {propertyValues {Curve {keys [
           {arriveTangent=-0.812, interpMode=RCIM_Cubic, leaveTangent=-0.812, tangentMode=RCTM_User, time=0, value=1.5}
           <at 1 1>
           <at 2 0.82>
           <at 100 0.75>
           <at 1000 1>
          ]}}}
       CurveIndex=$Emitter.Global Scale
      )
     )
    )
    Index 1 Param 3=&FloatFromCurve(
     FloatCurve=DataInterface() {propertyValues {Curve {keys [{interpMode=RCIM_Cubic, tangentMode=RCTM_User, time=0.4, value=0.5}, <at 0.8 0.1>]}}}
     Scale Curve=4
    )
    Index 1 Param 4=1
    Index 2 Param 1=&FloatFromCurve(
     FloatCurve=DataInterface() {propertyValues {Curve {keys [
         {arriveTangent=0.2023, interpMode=RCIM_Cubic, leaveTangent=0.2023, tangentMode=RCTM_User, time=0.3, value=0.1}
         {arriveTangent=0.5029, interpMode=RCIM_Cubic, leaveTangent=0.5029, tangentMode=RCTM_User, time=1, value=0.3}
        ]}}}
     Scale Curve=&Multiply_Float(
      A=&Multiply_Float(
       A=&FloatFromCurve(
        FloatCurve=DataInterface() {propertyValues {Curve {keys [
            {arriveTangent=1.5513, interpMode=RCIM_Cubic, leaveTangent=1.5513, tangentMode=RCTM_User, time=0.1, value=0}
            {arriveTangent=-6.1361, interpMode=RCIM_Cubic, leaveTangent=-6.1361, tangentMode=RCTM_User, time=1, value=3}
            <at 2 0.3>
            <at 10 0.1>
            <at 20 0.02>
            <at 50 0.002>
            <at 100 0.001>
            <at 200 0>
           ]}}}
        CurveIndex=$Emitter.Global Scale
       )
       B=$Emitter.Global Scale
      )
      B=$Emitter.Global Scale
     )
    )
    Index 2 Param 2=&FloatFromCurve(
     FloatCurve=DataInterface() {propertyValues {Curve {keys [
         {interpMode=RCIM_Cubic, tangentMode=RCTM_User, time=0.3, value=0}
         {arriveTangent=2.4333, interpMode=RCIM_Cubic, leaveTangent=2.4333, tangentMode=RCTM_User, time=1, value=1}
        ]}}}
     Scale Curve=0.15
    )
    Index 2 Param 3=&FloatFromCurve(
     FloatCurve=DataInterface() {propertyValues {Curve {keys [
         {arriveTangent=19.3692, interpMode=RCIM_Cubic, leaveTangent=19.3692, tangentMode=RCTM_User, time=0, value=0.25}
         {arriveTangent=-0.4167, interpMode=RCIM_Cubic, leaveTangent=-0.4167, time=0.1, value=1}
         {interpMode=RCIM_Cubic, leaveTangent=-0.7527, time=0.6, value=0}
        ]}}}
     Scale Curve=100
    )
    Index 2 Param 4=&FloatFromCurve(
     FloatCurve=DataInterface() {propertyValues {Curve {keys [
         {arriveTangent=-16.6453, interpMode=RCIM_Cubic, leaveTangent=-16.6453, tangentMode=RCTM_User, time=0, value=64}
         {arriveTangent=-87.1792, interpMode=RCIM_Cubic, leaveTangent=-87.1792, tangentMode=RCTM_User, time=0.1, value=24}
         {arriveTangent=-52.5, interpMode=RCIM_Cubic, leaveTangent=-52.5, time=0.2, value=12}
         {arriveTangent=-7.0531, interpMode=RCIM_Cubic, leaveTangent=-7.0531, tangentMode=RCTM_User, time=0.5, value=3}
         {arriveTangent=-0.6909, interpMode=RCIM_Cubic, leaveTangent=-0.6909, tangentMode=RCTM_User, time=1, value=2}
         <at 2 1>
         <at 3 0.3>
         <at 10 0.1>
         <at 20 0.05>
         <at 50 0.02>
         <at 100 0.01>
        ]}}}
     CurveIndex=$Emitter.Global Scale
    )
    Index 3 Param 1=&FloatFromCurve(
     FloatCurve=DataInterface() {propertyValues {Curve {keys [
         {arriveTangent=2.7705, interpMode=RCIM_Cubic, leaveTangent=2.7705, tangentMode=RCTM_User, time=0, value=0.2}
         {interpMode=RCIM_Cubic, tangentMode=RCTM_User, time=0.8, value=1}
        ]}}}
     Scale Curve=-2.5
    )
    Index 3 Param 2=&FloatFromCurve(
     FloatCurve=DataInterface() {propertyValues {Curve {keys [
         {arriveTangent=5.1563, interpMode=RCIM_Cubic, leaveTangent=5.1563, tangentMode=RCTM_User, time=0, value=0}
         {arriveTangent=1.4416, interpMode=RCIM_Cubic, leaveTangent=1.4416, tangentMode=RCTM_User, time=0.1699, value=0.4}
         {interpMode=RCIM_Cubic, tangentMode=RCTM_User, time=1, value=1.1}
        ]}}}
     Scale Curve=-0.15
    )
    Index 3 Param 3=&FloatFromCurve(
     FloatCurve=DataInterface() {propertyValues {Curve {keys [{arriveTangent=-127.3356, interpMode=RCIM_Cubic, time=0, value=65}, {interpMode=RCIM_Cubic, tangentMode=RCTM_User, time=1, value=0}]}}}
     Scale Curve=0.1
    )
    Index 3 Param 4=&FloatFromCurve(
     FloatCurve=DataInterface() {propertyValues {Curve {keys [{tangentMode=RCTM_User, time=0, value=0}, {tangentMode=RCTM_User, time=1, value=1}]}}}
     Scale Curve=-500
    )
    Module.Index0Param0Write=true
    Module.Index0Param1Write=true
    Module.Index0Param2Write=true
    Module.Index0Param3Write=true
    Module.Index1Param0Write=true
    Module.Index1Param1Write=true
    Module.Index1Param2Write=true
    Module.Index1Param3Write=true
    Module.Index2Param0Write=true
    Module.Index2Param1Write=true
    Module.Index2Param2Write=true
    Module.Index2Param3Write=true
    Module.Index3Param0Write=true
    Module.Index3Param1Write=true
    Module.Index3Param2Write=true
    Module.Index3Param3Write=true
    Write Parameter Index 1=true
    Write Parameter Index 2=true
    Write Parameter Index 3=true
   ) {module=DynamicMaterialParameters001}
   &ScaleColor(
    Linear Color Curve=DataInterface() {propertyValues {
      AlphaCurve {keys [
        {arriveTangent=13.8652, interpMode=RCIM_Cubic, leaveTangent=13.8652, tangentMode=RCTM_User, time=0, value=0}
        <at 0.2 1>
        {arriveTangentWeight=0.3303, interpMode=RCIM_Cubic, leaveTangentWeight=0.1386, tangentMode=RCTM_User, time=0.5, value=1}
        {arriveTangent=-4.5785, tangentMode=RCTM_User, time=0.9894, value=0}
       ]}
      BlueCurve {keys [<at 0.1 1.6952 RCIM_Cubic>,<at 0.479 0.6>]}
      GreenCurve {keys [<at 0.1 1.1598 RCIM_Cubic>,<at 0.479 0.6>]}
      RedCurve {keys [<at 0.1 2 RCIM_Cubic>,<at 0.479 0.6>]}
     }}
    Scale Mode=::NewEnumerator2
   )
   &CurlNoiseForce(
    Bias Noise=false
    Noise Quality / Cost=::NewEnumerator0
    Noise Strength=&RandomRangeFloat(Minimum=1200, Maximum=2000)
    Pan Noise=false
   )
   &SolveForcesAndVelocity(Clamp Velocity=false, Limit Acceleration=false)
   &ScaleMeshSize(Scale Factor=&VectorFromFloat(Value=&FloatFromCurve(FloatCurve=DataInterface() {propertyValues {Curve {
         defaultValue=1
         keys [
          {arriveTangent=1.221, interpMode=RCIM_Cubic, leaveTangent=1.221, tangentMode=RCTM_User, time=0, value=0.4}
          {interpMode=RCIM_Cubic, tangentMode=RCTM_User, time=1, value=1}
         ]
        }}})))
  ]
  rendererClasses [&NiagaraMeshRendererProperties]
  renderers [{rendererClass=&NiagaraMeshRendererProperties, rendererIndex=0}]
 }
 Preview {
  bEnabled=false
  emitterUpdateScript [
   &EmitterState()
   &SpawnRate(SpawnRate=500, Use Spawn Probability=false)
  ]
  particleSpawnScript [
   &PartitionParticles(Number Of Partitions=4)
   /SetVariables(Particles.ZPos:NiagaraFloat=&RandomRangeFloat(Minimum=0.2, Maximum=1)) {isSetParameters=true}
   &InitializeParticle(
    Color Mode=::NewEnumerator0
    Lifetime=2
    Position Offset=&SelectVectorFromArray(Direct Array Index=$Particles.PartitionParticles.Partition, Vector Selection Array=$System.TowerPositions)
    Position Offset Coordinate Space=::Local
    Sprite Size Mode=::NewEnumerator3
    Uniform Sprite Size=100
    UsePositionOffset=true
    Write Material Random=true
   )
   /SetVariables(Particles.Position=&AddVectorToPosition(
     Position=$Particles.Position
     Vector=&Multiply_Vector(
      A=&VectorFromCurve(
       VectorCurve=DataInterface() {propertyValues {
         XCurve {keys [<at 0 1>,<at 0.25 -1>,<at 0.5 -1>,<at 0.75 1>]}
         YCurve {keys [<at 0.25 1>,<at 0.5 -1>,<at 0.75 -1>,<at 1 1>]}
         ZCurve {keys [<at 0 1>,<at 1 1>]}
        }}
       CurveIndex=&RandomRangeFloat(Minimum=0, Maximum=1)
       Scale Curve=&Add_Vector(
        A=&Multiply_VectorByFloat(
         Vector=<1 1 0>
         Float=&Lerp_Float(
          A=1
          B=&SelectFloatFromArray(
           Direct Array Index=$Particles.PartitionParticles.Partition
           Float Selection Array=DataInterface() {propertyValues {FloatData [0.85,0.8,0.9,0.7]}}
          )
          Alpha=$Particles.ZPos
         )
        )
        B=&MakeVector(X=0, Y=0, Z=$Particles.ZPos)
       )
      )
      B=&SelectVectorFromArray(Direct Array Index=$Particles.PartitionParticles.Partition, Vector Selection Array=$System.TowerFloorSize)
     )
    )) {isSetParameters=true}
  ]
  particleUpdateScript [&ParticleState()]
  rendererClasses [&NiagaraSpriteRendererProperties]
  renderers [{rendererClass=&NiagaraSpriteRendererProperties}]
 }
 WallCrumbles {
  emitterUpdateScript [
   &EmitterState(Life Cycle Mode=::NewEnumerator1, Loop Behavior=::NewEnumerator1, Loop Duration=1.25, UseLoopDelay=false)
   &SpawnBurst_Instantaneous(Spawn Count=24, Use Spawn Probability=false, UseLoopCountLimit=false)
   &SpawnBurst_Instantaneous(Spawn Count=17, Spawn Time=0.04, Use Spawn Probability=false, UseLoopCountLimit=false) {module=SpawnBurst_Instantaneous001}
  ]
  particleSpawnScript [
   &PartitionParticles(
    Normalized Index In Partition Includes One=true
    Number Of Partitions=3
    Particles Per Partition=4
    Partition Mode:ENiagara_PartitionMode_v2=::NewEnumerator1
   )
   /SetVariables(Particles.ZPos:NiagaraFloat=&RandomRangeFloat(Minimum=0.8, Maximum=1)) {isSetParameters=true}
   /SetVariables(Particles.WallDirection:Vector3f=&RotateVectorByQuaternion(
     Vector To Rotate=<0 1 0>
     Quaternion=&MakeQuaternion(
      Quaternion From=0
      Axis=<0 0 1>
      Angle=&Multiply_FloatByInt(Float=0.25, Integer=$Particles.PartitionParticles.Index In Partition)
      From=?
      To=?
      Angle Type=::NewEnumerator1
      Normalize Vectors=false
     )
    )) {isSetParameters=true}
   &InitializeParticle(
    Color Maximum=<rgba 0.2805 0.3 0.261 1>
    Color Minimum=<rgba 0.1402 0.15 0.1305 1>
    Color Mode=::NewEnumerator2
    Lifetime Max=2.5
    Lifetime Min=1.7
    Lifetime Mode=::NewEnumerator1
    Mesh Scale Max=<3 5 5>
    Mesh Scale Min=<1 3 3>
    Mesh Scale Mode=::NewEnumerator2
    Position Offset=&Add_Vector(
     A=&SelectVectorFromArray(Direct Array Index=$Particles.PartitionParticles.Partition, Vector Selection Array=$System.TowerPositions)
     B=&MakeVector(
      X=0
      Y=0
      Z=&SelectFloatFromArray(Direct Array Index=$Particles.PartitionParticles.Partition, Float Selection Array=DataInterface() {propertyValues {FloatData [-150,1,400]}})
     )
    )
    Position Offset Coordinate Space=::Local
    UsePositionOffset=true
    Write Material Random=true
   )
   /SetVariables(Particles.Position=&AddVectorToPosition(
     Position=$Particles.Position
     Vector=&Multiply_Vector(
      A=&VectorFromCurve(
       VectorCurve=DataInterface() {propertyValues {
         XCurve {keys [<at 0 1>,<at 0.25 -1>,<at 0.5 -1>,<at 0.75 1>]}
         YCurve {keys [<at 0.25 1>,<at 0.5 -1>,<at 0.75 -1>,<at 1 1>]}
         ZCurve {keys [<at 0 1.5>,<at 1 1.5>]}
        }}
       CurveIndex=&Add_Float(
        A=&RandomRangeFloat(Minimum=0, Maximum=0.25)
        B=&Multiply_FloatByInt(Float=0.25, Integer=$Particles.PartitionParticles.Index In Partition)
       )
       Scale Curve=&Add_Vector(
        A=&Multiply_VectorByFloat(
         Vector=<1 1 0>
         Float=&Lerp_Float(
          A=1
          B=&SelectFloatFromArray(
           Direct Array Index=$Particles.PartitionParticles.Partition
           Float Selection Array=DataInterface() {propertyValues {FloatData [0.77,0.72,0.57]}}
          )
          Alpha=$Particles.ZPos
         )
        )
        B=&MakeVector(X=0, Y=0, Z=$Particles.ZPos)
       )
      )
      B=&SelectVectorFromArray(Direct Array Index=$Particles.PartitionParticles.Partition, Vector Selection Array=$System.TowerFloorSize)
     )
    )) {isSetParameters=true, module=SetVariables_#4}
   &AddVelocity(
    Velocity=&Add_Vector(A=$Particles.WallDirection, B=&RandomRangeVector(Minimum=<-0.1 -0.1 -0.1>, Maximum=<0.1 0.1 0.1>))
    Velocity Speed Scale=&RandomRangeFloat(Minimum=50, Maximum=250)
   )
   &InitialMeshOrientation(Mesh Orientation Mode=::NewEnumerator2, Orientation Vector=$Particles.WallDirection, Use Rotation Vector=false)
  ]
  particleUpdateScript [
   &ParticleState()
   &CameraOffset(Camera Offset Amount=&FloatFromCurve(
     FloatCurve=DataInterface() {propertyValues {Curve {
        defaultValue=1
        keys [
         {arriveTangent=5.2818, interpMode=RCIM_Cubic, leaveTangent=5.2818, tangentMode=RCTM_User, time=0, value=-0.1}
         {arriveTangent=1.5246, interpMode=RCIM_Cubic, leaveTangent=1.5246, tangentMode=RCTM_User, time=0.1, value=0.25}
         {arriveTangent=2.0926, tangentMode=RCTM_User, time=0.5, value=1}
        ]
       }}}
     Scale Curve=150
    ))
   &ScaleMeshSize(Scale Factor=&Add_Vector(
     A=<1 1 1>
     B=&VectorFromCurve(
      VectorCurve=DataInterface() {propertyValues {XCurve {keys [<at 0 0>,<at 1 1>]}, YCurve {keys [<at 0 0>,<at 1 1>]}, ZCurve {keys [<at 0 0>,<at 1 1>]}}}
      Scale Curve=&RandomRangeVector(Minimum=<5 1 2>, Maximum=<25 2.5 4>)
     )
    ))
   &DynamicMaterialParameters(
    Index 0 Param 1=&FloatFromCurve(FloatCurve=DataInterface() {propertyValues {Curve {
        defaultValue=1
        keys [
         {arriveTangent=32.5365, interpMode=RCIM_Cubic, leaveTangent=32.5365, tangentMode=RCTM_User, time=0, value=0}
         {arriveTangentWeight=0.1491, interpMode=RCIM_Cubic, leaveTangentWeight=0.2236, tangentMode=RCTM_User, time=0.05, value=1}
         {interpMode=RCIM_Cubic, tangentMode=RCTM_User, time=0.7, value=1}
         {arriveTangent=-4.1553, tangentMode=RCTM_User, time=1, value=0}
        ]
       }}})
    Index 0 Param 2=&RandomRangeFloat(Minimum=1, Maximum=4)
    Index 0 Param 3=&FloatFromCurve(
     FloatCurve=DataInterface() {propertyValues {Curve {keys [
         {arriveTangent=-2.3575, interpMode=RCIM_Cubic, leaveTangent=-2.3575, tangentMode=RCTM_User, time=0, value=1}
         {arriveTangent=-0.3909, tangentMode=RCTM_User, time=1, value=0}
        ]}}}
     Scale Curve=&RandomRangeFloat(Minimum=6, Maximum=12)
    )
    Module.Index0Float0Write=true
    Module.Index0Float1Write=true
    Module.Index0Float2Write=true
    Module.Index0Float3Write=false
   )
   &GravityForce(Gravity=&RandomRangeVector(Minimum=<0 0 -1200>, Maximum=<0 0 -600>))
   &SolveForcesAndVelocity(Clamp Velocity=false, Limit Acceleration=false)
  ]
  rendererClasses [&NiagaraMeshRendererProperties]
  renderers [{rendererClass=&NiagaraMeshRendererProperties, rendererIndex=0}]
 }
}
summary {
 systemName=NS_PistolCanary_Castle_GroundChurn
 emitters [
  {emitterName=WallCrumbles, rendererClasses [&NiagaraMeshRendererProperties]}
  {emitterName=MeshSmoke, rendererClasses [&NiagaraMeshRendererProperties]}
  {emitterName=LingerSmoke, rendererClasses [&NiagaraSpriteRendererProperties]}
  {emitterName=Bricks, rendererClasses [&NiagaraMeshRendererProperties]}
  {bEnabled=false, emitterName=Preview, rendererClasses [&NiagaraSpriteRendererProperties]}
  {emitterName=LowUnlitSmoke, rendererClasses [&NiagaraSpriteRendererProperties]}
  {emitterName=DebrisChunks, rendererClasses [&NiagaraMeshRendererProperties]}
 ]
 userVariables []
}
