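# Niagara system for PistolCanary's castle top bursting apart, throwing mesh debris and smoke outward and upward.
# BigMasks, GPU RoofTiles and WallCrumbles supply debris, alongside a magenta glow with light and sprite and mesh smoke layers.
# Tiles and crumbles split into four partitions, placed along wall directions scaled by System.UpperFloorSize.
# Spawn groups vary placement and speed, while curves drive tumbling, scaling, fades and dynamic material parameters.
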NiagaraSystem /PistolCanary/Assets/FX/Debris/NS_PistolCanary_Castle_TopBurst
dependencies {
 usedDataInterfaces [
  &NiagaraDataInterfaceCurve
  &NiagaraDataInterfaceVectorCurve
  &NiagaraDataInterfaceArrayFloat
  &NiagaraDataInterfaceArrayInt32
  &NiagaraDataInterfaceMeshRendererInfo
  &NiagaraDataInterfaceColorCurve
 ]
 usedDynamicInputs [&Vector2DFromFloat]
 usedModules [
  &self:SystemSpawnScript.SystemScriptSource.SystemScriptGraph.NiagaraNodeAssignment_1.SetVariables_#1
  &SystemState
 ]
}
emitters {
 BigMasks {
  emitterUpdateScript [
   &EmitterState()
   &SpawnBurst_Instantaneous(Use Spawn Probability=false, UseLoopCountLimit=false)
   &SpawnBurst_Instantaneous(Spawn Group=1, Use Spawn Probability=false, UseLoopCountLimit=false) {module=SpawnBurst_Instantaneous001}
  ]
  particleSpawnScript [
   &InitializeParticle(
    Color Mode=::NewEnumerator0
    Lifetime=4
    Mesh Scale Mode=::NewEnumerator3
    Position=&MaskVectorBySpawnGroup(Masked Vector=<14333 71471 7931>, Passthrough Vector=<14339 69320 7938>)
    Position Mode=::NewEnumerator1
    UsePositionOffset=false
    Write Material Random=true
   )
   &AddVelocity(Module.Use Falloff Distance=false, Origin Offset=<0 0 -1000>, Override Seed=false, Velocity Mode=::NewEnumerator1, Velocity Speed=1500)
   &InitialMeshOrientation(
    Mesh Orientation Mode=::NewEnumerator2
    Orientation Vector=&MaskVectorBySpawnGroup(Masked Vector=<1 0 0>, Passthrough Vector=<-1 0 0>)
    Use Rotation Vector=false
   )
  ]
  particleUpdateScript [
   &ParticleState()
   &ScaleMeshSize(Scale Factor=&VectorFromFloat(Value=&FloatFromCurve(FloatCurve=DataInterface() {propertyValues {Curve {keys [<at 0.85 1 RCIM_Cubic>, {arriveTangent=-13.3021, tangentMode=RCTM_User, time=1, value=0}]}}})))
   &Drag(Drag=0.25, Use Linear Drag=true, Use Rotational Drag=false)
   &GravityForce(Gravity=<0 0 -1200>)
   &SolveForcesAndVelocity(Clamp Velocity=false, Limit Acceleration=false)
   &UpdateMeshOrientation(
    Rotation Coordinate Space=::NewEnumerator1
    Rotation Rate=&FloatFromCurve(
     FloatCurve=DataInterface() {propertyValues {Curve {
        defaultValue=1
        keys [
         {arriveTangent=-0.9725, time=0, value=1}
         {arriveTangent=-0.5583, interpMode=RCIM_Cubic, leaveTangent=-0.5583, tangentMode=RCTM_User, time=1, value=0.75}
        ]
       }}}
     CurveIndex=$Particles.Age
     Scale Curve=-0.5
    )
    Rotation Vector=&Add_Vector(
     A=&MaskVectorBySpawnGroup(Masked Vector=<1 0 0>, Passthrough Vector=<-1 0 0>)
     B=&RandomVector(Vector Scale=&RandomRangeFloat(Minimum=0.25, Maximum=0.35))
    )
   )
  ]
  rendererClasses [&NiagaraMeshRendererProperties]
  renderers [{rendererClass=&NiagaraMeshRendererProperties, rendererIndex=0}]
 }
 GlowBurst {
  emitterUpdateScript [
   &EmitterState()
   &SpawnBurst_Instantaneous(Use Spawn Probability=false, UseLoopCountLimit=false)
  ]
  particleSpawnScript [
   &InitializeParticle(
    Color=<rgba 2 0.1443 1.9866 1>
    Lifetime=1.5
    Sprite Size Mode=::NewEnumerator3
    Uniform Sprite Size=5000
    UsePositionOffset=false
    Write Material Random=true
   )
   &DynamicMaterialParameters(
    Index 0 Param 1=3
    Index 0 Param 2=50
    Module.Index0Float0Write=true
    Module.Index0Float1Write=true
    Module.Index0Float2Write=false
    Module.Index0Float3Write=false
   )
   &CameraOffset(Camera Offset Amount=1000)
  ]
  particleUpdateScript [
   &ParticleState()
   &ScaleSpriteSize(Uniform Curve Sprite Scale=DataInterface() {propertyValues {Curve {
       defaultValue=0
       keys [
        {arriveTangent=17.8836, interpMode=RCIM_Cubic, leaveTangent=17.8836, tangentMode=RCTM_User, time=0, value=0.6}
        {interpMode=RCIM_Cubic, tangentMode=RCTM_User, time=0.05, value=1}
        {arriveTangent=-2.0463, tangentMode=RCTM_User, time=0.5, value=0}
       ]
      }}})
   &ScaleColor(
    Scale Alpha=&FloatFromCurve(FloatCurve=DataInterface() {propertyValues {Curve {
        defaultValue=1
        keys [
         {arriveTangent=30.8891, interpMode=RCIM_Cubic, leaveTangent=30.8891, tangentMode=RCTM_User, time=0, value=0}
         {arriveTangent=-2.5249, interpMode=RCIM_Cubic, leaveTangent=-2.5249, tangentMode=RCTM_User, time=0.1, value=1}
         {arriveTangent=-0.0405, interpMode=RCIM_Cubic, leaveTangent=-0.0405, tangentMode=RCTM_User, time=1, value=0}
        ]
       }}})
    ScaleA=true
    ScaleRGB=false
   )
   /SetVariables(Particles.LightColor=&Multiply_LinearColorByFloat(
     Linear Color=$Particles.Initial.Color
     Float=&FloatFromCurve(
      FloatCurve=DataInterface() {propertyValues {Curve {
         defaultValue=1
         keys [
          {arriveTangent=34.9316, interpMode=RCIM_Cubic, leaveTangent=34.9316, tangentMode=RCTM_User, time=0, value=0}
          {interpMode=RCIM_Cubic, tangentMode=RCTM_User, time=0.1, value=1}
          {tangentMode=RCTM_User, time=1, value=0}
         ]
        }}}
      Scale Curve=16
     )
    )) {isSetParameters=true}
   /SetVariables(Particles.Position=&MakeVector(
     X=0
     Y=0
     Z=&FloatFromCurve(
      FloatCurve=DataInterface() {propertyValues {Curve {
         defaultValue=1
         keys [
          {arriveTangent=3.4235, interpMode=RCIM_Cubic, leaveTangent=3.4235, tangentMode=RCTM_User, time=0, value=-0.7}
          {arriveTangent=0.4852, tangentMode=RCTM_User, time=1, value=1}
         ]
        }}}
      Scale Curve=2000
     )
    )) {isSetParameters=true}
  ]
  rendererClasses [&NiagaraSpriteRendererProperties, &NiagaraLightRendererProperties]
  renderers [
   {rendererClass=&NiagaraSpriteRendererProperties}
   {rendererClass=&NiagaraLightRendererProperties, rendererIndex=1}
  ]
 }
 LingerSmoke {
  emitterUpdateScript [
   &EmitterState()
   &SpawnRate(SpawnRate=12, Use Spawn Probability=false)
  ]
  particleSpawnScript [
   &InitializeParticle(
    Color=<rgba 0.02 0.019 0.0181 1.5>
    Lifetime=&RandomRangeFloat(Minimum=3, Maximum=4)
    Position Offset=&MakeVector(
     X=0
     Y=0
     Z=&Lerp_Float(
      A=-500
      B=-2000
      Alpha=&FloatFromCurve(
       FloatCurve=DataInterface() {propertyValues {Curve {defaultValue=1, keys [<at 0 0>, {arriveTangent=0.8379, tangentMode=RCTM_User, time=3, value=1}]}}}
       CurveIndex=$Emitter.Age
      )
     )
    )
    Sprite Rotation Mode=::NewEnumerator2
    Sprite Size Mode=::NewEnumerator3
    Uniform Sprite Size=&RandomRangeFloat(Minimum=2000, Maximum=3000)
    UsePositionOffset=true
    Write Material Random=true
   )
   &ShapeLocation(
    Cylinder Height=1000
    Cylinder Radius=1000
    MaskSpawnGroup=false
    Override Seed=false
    Shape Primitive=::NewEnumerator1
    Surface Only=false
    UseOwnerScale=true
   )
   /SetVariables(Particles.InitSpriteRotRate=&UniformRangedFloat(Minimum=-3, Maximum=3)) {isSetParameters=true}
   /SetVariables(Particles.RandomFloat=&RandomRangeFloat(Minimum=0, Maximum=1)) {isSetParameters=true}
   /SetVariables(Particles.SubUVRandomStart:NiagaraInt32=&RandomRangeInt(Minimum=0, Maximum=44)) {isSetParameters=true}
   &AddVelocity(
    Module.Use Falloff Distance=false
    Origin Offset=<0 0 -2000>
    Override Seed=false
    Velocity Mode=::NewEnumerator1
    Velocity Speed=&RandomRangeFloat(Minimum=600, Maximum=900)
   )
  ]
  particleUpdateScript [
   &ParticleState()
   &WindForce(Turbulence Mode:ENiagaraWindTurbulenceMode=::NewEnumerator0, Wind Speed=<0 0 1>, Wind Speed Scale=500)
   &Drag(Drag=0.5, Use Linear Drag=true, Use Rotational Drag=false)
   &SolveForcesAndVelocity(Clamp Velocity=false, Limit Acceleration=false)
   &ScaleColor(
    Scale Alpha=&FloatFromCurve(FloatCurve=DataInterface() {propertyValues {Curve {keys [
         {arriveTangent=43.8033, interpMode=RCIM_Cubic, time=0, value=0}
         <at 0.25 1 RCIM_Cubic>
         {arriveTangent=-2.8482, interpMode=RCIM_Cubic, leaveTangent=-2.8482, tangentMode=RCTM_User, time=1, value=0}
        ]}}})
    ScaleA=true
    ScaleRGB=false
   )
   &SpriteRotationRate(Rotation Rate=&Multiply_Float(
     A=$Particles.InitSpriteRotRate
     B=&FloatFromCurve(
      FloatCurve=DataInterface() {propertyValues {Curve {keys [
          {arriveTangent=-2.157, interpMode=RCIM_Cubic, leaveTangent=-2.157, tangentMode=RCTM_User, time=-0.0046, value=1.0113}
          {interpMode=RCIM_Cubic, leaveTangent=-0.2496, time=1, value=0.16}
         ]}}}
      Scale Curve=20
     )
    ))
   &ScaleSpriteSize(Uniform Curve Sprite Scale=DataInterface() {propertyValues {Curve {keys [
        {arriveTangent=1.5213, interpMode=RCIM_Cubic, leaveTangent=1.5213, tangentMode=RCTM_User, time=0, value=0.4}
        {arriveTangent=0.3088, tangentMode=RCTM_User, time=1, value=1}
       ]}}})
   &CameraOffset(Camera Offset Amount=/Niagara/DynamicInputs/Math/CalculateParticleRadius(
     Radius Calculation Type:ENiagaraCollisionRadiusOptions=::NewEnumerator0
     Particle Radius:NiagaraFloat=1
     Mesh Dimensions:Vector3f=<1 1 1>
     Method for Calculating Particle Radius:ENiagaraCalculateRadiusOptions=::NewEnumerator2
    ))
   /SetVariables(Particles.SubImageIndex=&Lerp_Float(A=&RandomRangeFloat(Minimum=4, Maximum=12), B=63, Alpha=$Particles.NormalizedAge)) {isSetParameters=true}
  ]
  rendererClasses [&NiagaraSpriteRendererProperties]
  renderers [{rendererClass=&NiagaraSpriteRendererProperties}]
 }
 LowUnlitSmoke {
  emitterUpdateScript [
   &EmitterState()
   &SpawnBurst_Instantaneous(Spawn Count=16, Use Spawn Probability=false, UseLoopCountLimit=false)
   &SpawnRate(
    SpawnGroup=1
    SpawnRate=&FloatFromCurve(
     FloatCurve=DataInterface() {propertyValues {Curve {defaultValue=0, keys [<at 0 1 RCIM_Cubic>, {arriveTangent=-2.108, tangentMode=RCTM_User, time=1, value=0}]}}}
     CurveIndex=$Emitter.NormalizedLoopAge
    )
    Use Spawn Probability=false
   )
  ]
  particleSpawnScript [
   &PartitionParticles(
    Normalized Index In Partition Includes One=true
    Number Of Partitions=4
    Particles Per Partition=5
    Partition Mode:ENiagara_PartitionMode_v2=::NewEnumerator1
   )
   &InitializeParticle(
    Color Maximum=<rgba 0.05 0.0278 0.0331 1.5>
    Color Minimum=<rgba 0.04 0.0255 0.0314 1.5>
    Color Mode=::NewEnumerator2
    Lifetime=&RandomRangeFloat(Minimum=1.5, Maximum=2.5)
    Mesh Scale=&RandomRangeVector(Minimum=<2.5 2 2>, Maximum=<3 3.5 3.5>)
    Mesh Scale Mode=::NewEnumerator4
    Position Offset=<0 0 -500>
    Sprite Rotation Mode=::NewEnumerator2
    Sprite Size Mode=::NewEnumerator1
    Uniform Sprite Size Max=2500
    Uniform Sprite Size Min=1700
    UsePositionOffset=true
    Write Material Random=true
   )
   &ShapeLocation(
    Cylinder Distribution=::NewEnumerator2
    Cylinder Height=800
    Cylinder Radius=800
    MaskSpawnGroup=false
    Shape Primitive=::NewEnumerator1
    U Position=&MaskFloatBySpawnGroup(
     Masked Float=&ReturnNormalizedExecIndex(Normalized Index Includes One=false)
     Passthrough Float=&RandomRangeFloat(Minimum=0, Maximum=1)
    )
    UseOwnerScale=true
    V Position=&RandomRangeFloat(Minimum=0, Maximum=1)
   )
   &AddVelocity(
    Module.Use Falloff Distance=false
    Origin Offset=<0 0 -1000>
    Override Seed=false
    Velocity Mode=::NewEnumerator1
    Velocity Speed=&Multiply_Float(
     A=&RandomRangeFloat(Minimum=0.5, Maximum=1)
     B=&MaskFloatBySpawnGroup(Masked Float=2500, Passthrough Float=500)
    )
   ) {module=AddVelocity001}
   &CameraOffset(Camera Offset Amount=500)
  ]
  particleUpdateScript [
   &ParticleState()
   &ScaleSpriteSize(Uniform Curve Sprite Scale=DataInterface() {propertyValues {Curve {
       defaultValue=1
       keys [
        {arriveTangent=1.221, interpMode=RCIM_Cubic, leaveTangent=1.221, tangentMode=RCTM_User, time=0, value=0.4}
        {interpMode=RCIM_Cubic, tangentMode=RCTM_User, time=1, value=1}
       ]
      }}})
   &Drag(Drag=&RandomRangeFloat(Minimum=2, Maximum=3), Use Linear Drag=true, Use Rotational Drag=false)
   &ScaleColor(
    Scale Alpha=&FloatFromCurve(FloatCurve=DataInterface() {propertyValues {Curve {
        defaultValue=1
        keys [
         {arriveTangent=8.3954, interpMode=RCIM_Cubic, leaveTangent=8.3954, tangentMode=RCTM_User, time=0, value=0}
         {
          arriveTangent=-0.5026
          arriveTangentWeight=0.1491
          interpMode=RCIM_Cubic
          leaveTangent=-0.5026
          leaveTangentWeight=0.2236
          tangentMode=RCTM_User
          time=0.2
          value=1.25
         }
         {arriveTangentWeight=0.1, interpMode=RCIM_Cubic, leaveTangentWeight=0.1667, tangentMode=RCTM_User, time=0.75, value=1}
         {arriveTangent=-6.06, tangentMode=RCTM_User, time=1, value=0}
        ]
       }}})
    ScaleA=true
    ScaleRGB=false
   )
   &SolveForcesAndVelocity(Clamp Velocity=false, Limit Acceleration=false)
   &DynamicMaterialParameters(
    Index 0 Param 1=&Add_Float(
     A=$Particles.MaterialRandom
     B=&FloatFromCurve(FloatCurve=DataInterface() {propertyValues {Curve {keys [
          {arriveTangent=2.7533, interpMode=RCIM_Cubic, leaveTangent=1.8958, tangentMode=RCTM_User, time=0, value=1}
          {arriveTangent=0.4122, interpMode=RCIM_Cubic, leaveTangent=0.4122, tangentMode=RCTM_User, time=0.321, value=1.3453}
          {arriveTangent=0.1513, interpMode=RCIM_Cubic, leaveTangent=0.1513, tangentMode=RCTM_User, time=1, value=1.5}
         ]}}})
    )
    Index 0 Param 3=&FloatFromCurve(FloatCurve=DataInterface() {propertyValues {Curve {keys [{arriveTangent=-9.6776, interpMode=RCIM_Cubic, leaveTangent=-9.6776, tangentMode=RCTM_User, time=0, value=1}, <at 0.2 0.05>, <at 0.5 0>]}}})
    Index 0 Param 4=&FloatFromCurve(FloatCurve=DataInterface() {propertyValues {Curve {keys [<at 0.5 1 RCIM_Cubic>,<at 1 0>]}}})
    Index 2 Param 1=&FloatFromCurve(FloatCurve=DataInterface() {propertyValues {Curve {keys [
         {interpMode=RCIM_Cubic, leaveTangent=4.2485, tangentMode=RCTM_Break, time=0, value=0}
         {arriveTangent=1.0548, interpMode=RCIM_Cubic, leaveTangent=1.0548, tangentMode=RCTM_User, time=0.17, value=0.5715}
         {arriveTangent=0.4121, tangentMode=RCTM_User, time=1, value=1}
        ]}}})
    Index 2 Param 2=&FloatFromCurve(
     FloatCurve=DataInterface() {propertyValues {Curve {keys [
         {arriveTangent=0.4242, interpMode=RCIM_Cubic, leaveTangent=0.4242, tangentMode=RCTM_User, time=0, value=0.1}
         {arriveTangent=0.1815, interpMode=RCIM_Cubic, leaveTangent=0.1815, tangentMode=RCTM_User, time=1, value=0.4}
        ]}}}
     Scale Curve=&RandomRangeFloat(Minimum=0.5, Maximum=1)
    )
    Index 2 Param 3=&FloatFromCurve(
     FloatCurve=DataInterface() {propertyValues {Curve {keys [
         {interpMode=RCIM_Cubic, leaveTangent=1.2101, tangentMode=RCTM_Break, time=0, value=0.25}
         {arriveTangent=0.6182, interpMode=RCIM_Cubic, leaveTangent=11.1233, tangentMode=RCTM_Break, time=1, value=1}
        ]}}}
     Scale Curve=3
    )
    Index 2 Param 4=&FloatFromCurve(
     FloatCurve=DataInterface() {propertyValues {Curve {keys [
         {arriveTangent=-0.5563, interpMode=RCIM_Cubic, leaveTangent=-0.5563, tangentMode=RCTM_User, time=0, value=1}
         {interpMode=RCIM_Cubic, leaveTangent=-1.2956, tangentMode=RCTM_Break, time=1, value=0.75}
        ]}}}
     Scale Curve=2
    )
    Module.Index0Float0Write=true
    Module.Index0Float1Write=false
    Module.Index0Float2Write=true
    Module.Index0Float3Write=true
    Module.Index2Float0Write=true
    Module.Index2Float1Write=true
    Module.Index2Float2Write=true
    Module.Index2Float3Write=true
    Write Parameter Index 2=::NewEnumerator1
   )
  ]
  rendererClasses [&NiagaraSpriteRendererProperties]
  renderers [{rendererClass=&NiagaraSpriteRendererProperties}]
 }
 MeshSmoke {
  emitterSpawnScript [
   /SetVariables(Emitter.Global Color:LinearColor=<rgba 0.3894 0.0859 1 1>, Emitter.Global Scale=5) {isSetParameters=true}
   /SetVariables(
    Emitter.Color2:LinearColor=&MultiplyLinearColor(Linear Color=$Emitter.Global Color, Scale Factor=<2 1 1 1>)
    Emitter.Depth Fade:NiagaraFloat=&Multiply_Float(A=50, B=$Emitter.Global Scale)
    Emitter.Dissolve Noise Power=1.1
    Emitter.Dissolve Noise Scale:LinearColor=&MultiplyLinearColor(
     Linear Color=/Niagara/DynamicInputs/TypeConversions/MakeLinearColorFromFloat(Value:NiagaraFloat=&FloatFromCurve(FloatCurve=DataInterface() {propertyValues {Curve {keys [<at 0 2>,<at 1 1>,<at 2 0.5>]}}}, CurveIndex=$Emitter.Global Scale))
     Scale Factor=<0.0008 0.0008 0.0009 1>
    )
    Emitter.Near Camera Fade:NiagaraFloat=&Multiply_Float(A=200, B=$Emitter.Global Scale)
   ) {isSetParameters=true}
   /SetVariables(Emitter.Dissolve Scale Adjsutment:NiagaraFloat=&FloatFromCurve(
     FloatCurve=DataInterface() {propertyValues {Curve {keys [<at 0 3>,<at 1 1>,<at 5 0.5>,<at 20 0.4>,<at 100 0.4>]}}}
     CurveIndex=$Emitter.Global Scale
     Scale Curve=$Emitter.Global Scale
    )) {isSetParameters=true}
  ]
  emitterUpdateScript [
   &EmitterState()
   &SpawnBurst_Instantaneous(Spawn Count=16, Use Spawn Probability=false, UseLoopCountLimit=false)
   &SpawnRate(
    SpawnGroup=1
    SpawnRate=&FloatFromCurve(
     FloatCurve=DataInterface() {propertyValues {Curve {defaultValue=0, keys [<at 0 1 RCIM_Cubic>, {arriveTangent=-2.108, tangentMode=RCTM_User, time=1, value=0}]}}}
     CurveIndex=$Emitter.NormalizedLoopAge
    )
    Use Spawn Probability=false
   )
  ]
  particleSpawnScript [
   &PartitionParticles(
    Normalized Index In Partition Includes One=true
    Number Of Partitions=4
    Particles Per Partition=5
    Partition Mode:ENiagara_PartitionMode_v2=::NewEnumerator1
   )
   &InitializeParticle(
    Color=<rgba 0.12 0.1143 0.1088 1.6>
    Lifetime=&RandomRangeFloat(Minimum=1.5, Maximum=2.2)
    Mesh Scale=&RandomRangeVector(Minimum=<2.5 2 2>, Maximum=<3 3.5 3.5>)
    Mesh Scale Mode=::NewEnumerator4
    Position Offset=<0 0 -500>
    UsePositionOffset=true
    Write Material Random=true
   )
   &ShapeLocation(
    Cylinder Distribution=::NewEnumerator2
    Cylinder Height=800
    Cylinder Radius=800
    MaskSpawnGroup=false
    Shape Primitive=::NewEnumerator1
    U Position=&MaskFloatBySpawnGroup(
     Masked Float=&ReturnNormalizedExecIndex(Normalized Index Includes One=false)
     Passthrough Float=&RandomRangeFloat(Minimum=0, Maximum=1)
    )
    UseOwnerScale=true
    V Position=&RandomRangeFloat(Minimum=0, Maximum=1)
   )
   &AddVelocity(
    Module.Use Falloff Distance=false
    Origin Offset=<0 0 -1000>
    Override Seed=false
    Velocity Mode=::NewEnumerator1
    Velocity Speed=&Multiply_Float(
     A=&RandomRangeFloat(Minimum=0.5, Maximum=1)
     B=&MaskFloatBySpawnGroup(Masked Float=2500, Passthrough Float=500)
    )
   ) {module=AddVelocity001}
   &CameraOffset(Camera Offset Amount=400)
   /SetVariables(Particles.Scale=&Multiply_VectorByFloat(Vector=$Particles.Scale, Float=$Emitter.Global Scale)) {isSetParameters=true}
  ]
  particleUpdateScript [
   &ParticleState()
   &DynamicMaterialParameters(
    Index 0 Param 1=&Multiply_Float(A=-100, B=$Emitter.Global Scale)
    Index 0 Param 2=&FloatFromCurve(FloatCurve=DataInterface() {propertyValues {Curve {keys [
         {tangentMode=RCTM_User, time=0, value=1}
         {arriveTangent=-0.0405, interpMode=RCIM_Cubic, leaveTangent=-0.0405, tangentMode=RCTM_User, time=0.4, value=1}
         {arriveTangent=-1.6592, interpMode=RCIM_Cubic, leaveTangent=-1.6592, tangentMode=RCTM_User, time=1, value=0.4}
        ]}}})
    Index 0 Param 3=&FloatFromCurve(
     FloatCurve=DataInterface() {propertyValues {Curve {keys [
         {arriveTangent=1.4725, interpMode=RCIM_Cubic, leaveTangent=1.4725, tangentMode=RCTM_User, time=0, value=0}
         {arriveTangent=0.4975, tangentMode=RCTM_User, time=1, value=1}
        ]}}}
     Scale Curve=&Multiply_Float(
      A=&RandomRangeFloat(Minimum=100, Maximum=150)
      B=&Multiply_Float(
       A=$Emitter.Global Scale
       B=&FloatFromCurve(FloatCurve=DataInterface() {propertyValues {Curve {keys [<at 0 1>,<at 100 0.9>,<at 300 0>]}}}, CurveIndex=$Emitter.Global Scale)
      )
     )
    )
    Index 0 Param 4=&FloatFromCurve(
     FloatCurve=DataInterface() {propertyValues {Curve {keys [
         {arriveTangent=0.1175, interpMode=RCIM_Cubic, leaveTangent=1.7315, tangentMode=RCTM_Break, time=0.1, value=-3}
         {arriveTangent=21.7158, interpMode=RCIM_Cubic, leaveTangent=21.7158, tangentMode=RCTM_User, time=1, value=6}
        ]}}}
     Scale Curve=0.5
    )
    Index 1 Param 1=&FloatFromCurve(
     FloatCurve=DataInterface() {propertyValues {Curve {keys [
         {arriveTangent=-31.7902, interpMode=RCIM_Cubic, time=0, value=5}
         {arriveTangent=4.4473, interpMode=RCIM_Cubic, leaveTangent=4.4473, tangentMode=RCTM_User, time=0.15, value=0.7}
         {arriveTangent=16.9091, interpMode=RCIM_Cubic, leaveTangent=16.9091, time=0.4, value=7}
         <at 0.7 10 RCIM_Cubic>
        ]}}}
     Scale Curve=2
    )
    Index 1 Param 2=&Divide_Float(
     A=&RandomRangeFloat(Minimum=0.016, Maximum=0.02)
     B=&Multiply_Float(
      A=$Emitter.Global Scale
      B=&FloatFromCurve(
       FloatCurve=DataInterface() {propertyValues {Curve {keys [
           {arriveTangent=-0.812, interpMode=RCIM_Cubic, leaveTangent=-0.812, tangentMode=RCTM_User, time=0, value=1.5}
           <at 1 1>
           <at 2 0.82>
           <at 100 0.75>
           <at 1000 1>
          ]}}}
       CurveIndex=$Emitter.Global Scale
      )
     )
    )
    Index 1 Param 3=&FloatFromCurve(
     FloatCurve=DataInterface() {propertyValues {Curve {keys [{interpMode=RCIM_Cubic, tangentMode=RCTM_User, time=0.4, value=0.5}, <at 0.8 0.1>]}}}
     Scale Curve=5
    )
    Index 1 Param 4=1
    Index 2 Param 1=&FloatFromCurve(
     FloatCurve=DataInterface() {propertyValues {Curve {keys [
         {arriveTangent=0.2023, interpMode=RCIM_Cubic, leaveTangent=0.2023, tangentMode=RCTM_User, time=0.3, value=0.1}
         {arriveTangent=0.5029, interpMode=RCIM_Cubic, leaveTangent=0.5029, tangentMode=RCTM_User, time=1, value=0.3}
        ]}}}
     Scale Curve=&Multiply_Float(
      A=&Multiply_Float(
       A=&FloatFromCurve(
        FloatCurve=DataInterface() {propertyValues {Curve {keys [
            {arriveTangent=1.5513, interpMode=RCIM_Cubic, leaveTangent=1.5513, tangentMode=RCTM_User, time=0.1, value=0}
            {arriveTangent=-6.1361, interpMode=RCIM_Cubic, leaveTangent=-6.1361, tangentMode=RCTM_User, time=1, value=3}
            <at 2 0.3>
            <at 10 0.1>
            <at 20 0.02>
            <at 50 0.002>
            <at 100 0.001>
            <at 200 0>
           ]}}}
        CurveIndex=$Emitter.Global Scale
       )
       B=$Emitter.Global Scale
      )
      B=$Emitter.Global Scale
     )
    )
    Index 2 Param 2=&FloatFromCurve(
     FloatCurve=DataInterface() {propertyValues {Curve {keys [
         {interpMode=RCIM_Cubic, tangentMode=RCTM_User, time=0.3, value=0}
         {arriveTangent=2.4333, interpMode=RCIM_Cubic, leaveTangent=2.4333, tangentMode=RCTM_User, time=1, value=1}
        ]}}}
     Scale Curve=0.05
    )
    Index 2 Param 3=&FloatFromCurve(
     FloatCurve=DataInterface() {propertyValues {Curve {keys [
         {arriveTangent=14.5529, interpMode=RCIM_Cubic, leaveTangent=14.5529, tangentMode=RCTM_User, time=0, value=0}
         {arriveTangent=-5.1575, interpMode=RCIM_Cubic, leaveTangent=-5.1575, tangentMode=RCTM_User, time=0.15, value=1}
         {interpMode=RCIM_Cubic, leaveTangent=-0.7527, time=0.5, value=0}
        ]}}}
     Scale Curve=400
    )
    Index 2 Param 4=&FloatFromCurve(
     FloatCurve=DataInterface() {propertyValues {Curve {keys [
         {arriveTangent=-16.6453, interpMode=RCIM_Cubic, leaveTangent=-16.6453, tangentMode=RCTM_User, time=0, value=64}
         {arriveTangent=-87.1792, interpMode=RCIM_Cubic, leaveTangent=-87.1792, tangentMode=RCTM_User, time=0.1, value=24}
         {arriveTangent=-52.5, interpMode=RCIM_Cubic, leaveTangent=-52.5, time=0.2, value=12}
         {arriveTangent=-7.0531, interpMode=RCIM_Cubic, leaveTangent=-7.0531, tangentMode=RCTM_User, time=0.5, value=3}
         {arriveTangent=-0.6909, interpMode=RCIM_Cubic, leaveTangent=-0.6909, tangentMode=RCTM_User, time=1, value=2}
         <at 2 1>
         <at 3 0.3>
         <at 10 0.1>
         <at 20 0.05>
         <at 50 0.02>
         <at 100 0.01>
        ]}}}
     CurveIndex=$Emitter.Global Scale
    )
    Index 3 Param 1=&FloatFromCurve(
     FloatCurve=DataInterface() {propertyValues {Curve {keys [
         {arriveTangent=2.2353, interpMode=RCIM_Cubic, leaveTangent=2.2353, tangentMode=RCTM_User, time=0, value=0.2}
         {arriveTangent=1.4416, interpMode=RCIM_Cubic, leaveTangent=1.4416, tangentMode=RCTM_User, time=0.2, value=0.6}
         {interpMode=RCIM_Cubic, tangentMode=RCTM_User, time=0.8, value=1.25}
        ]}}}
     Scale Curve=-3
    )
    Index 3 Param 2=&FloatFromCurve(
     FloatCurve=DataInterface() {propertyValues {Curve {keys [
         {arriveTangent=5.1563, interpMode=RCIM_Cubic, leaveTangent=5.1563, tangentMode=RCTM_User, time=0, value=0}
         {arriveTangent=1.4416, interpMode=RCIM_Cubic, leaveTangent=1.4416, tangentMode=RCTM_User, time=0.1699, value=0.4}
         {interpMode=RCIM_Cubic, tangentMode=RCTM_User, time=1, value=1.1}
        ]}}}
     Scale Curve=-1
    )
    Index 3 Param 3=&FloatFromCurve(FloatCurve=DataInterface() {propertyValues {Curve {keys [
         {arriveTangent=-11.5239, interpMode=RCIM_Cubic, leaveTangent=-11.5239, tangentMode=RCTM_User, time=0.04, value=65}
         {arriveTangent=-43.7054, interpMode=RCIM_Cubic, leaveTangent=-43.7054, tangentMode=RCTM_User, time=0.35, value=2}
         {interpMode=RCIM_Cubic, tangentMode=RCTM_User, time=0.5, value=0}
        ]}}})
    Index 3 Param 4=&FloatFromCurve(
     FloatCurve=DataInterface() {propertyValues {Curve {keys [{tangentMode=RCTM_User, time=0, value=0}, {tangentMode=RCTM_User, time=1, value=1}]}}}
     Scale Curve=-1500
    )
    Module.Index0Param0Write=true
    Module.Index0Param1Write=true
    Module.Index0Param2Write=true
    Module.Index0Param3Write=true
    Module.Index1Param0Write=true
    Module.Index1Param1Write=true
    Module.Index1Param2Write=true
    Module.Index1Param3Write=true
    Module.Index2Param0Write=true
    Module.Index2Param1Write=true
    Module.Index2Param2Write=true
    Module.Index2Param3Write=true
    Module.Index3Param0Write=true
    Module.Index3Param1Write=true
    Module.Index3Param2Write=true
    Module.Index3Param3Write=true
    Write Parameter Index 1=true
    Write Parameter Index 2=true
    Write Parameter Index 3=true
   ) {module=DynamicMaterialParameters001}
   &ScaleColor(
    Linear Color Curve=DataInterface() {propertyValues {
      AlphaCurve {keys [
        {arriveTangent=13.8652, interpMode=RCIM_Cubic, leaveTangent=13.8652, tangentMode=RCTM_User, time=0, value=0}
        <at 0.2 1>
        {arriveTangent=-1.2668, arriveTangentWeight=0.3303, interpMode=RCIM_Cubic, leaveTangent=-1.2668, leaveTangentWeight=0.1386, time=0.6, value=1}
        {arriveTangent=-5.2846, tangentMode=RCTM_User, time=0.9894, value=0}
       ]}
      BlueCurve {keys [
        <at 0 2 RCIM_Cubic>
        {arriveTangentWeight=0.0858, interpMode=RCIM_Cubic, leaveTangentWeight=0.456, tangentMode=RCTM_User, time=0.1888, value=0.8}
        <at 0.6 0.4497>
       ]}
      GreenCurve {keys [
        <at 0 2 RCIM_Cubic>
        {arriveTangentWeight=0.0858, interpMode=RCIM_Cubic, leaveTangentWeight=0.456, tangentMode=RCTM_User, time=0.1888, value=0.8}
        <at 0.6 0.4497>
       ]}
      RedCurve {keys [
        <at 0 2 RCIM_Cubic>
        {arriveTangentWeight=0.0858, interpMode=RCIM_Cubic, leaveTangentWeight=0.456, tangentMode=RCTM_User, time=0.1888, value=0.8}
        <at 0.6 0.4497>
       ]}
     }}
    Scale Mode=::NewEnumerator2
   )
   &WindForce(Turbulence Mode:ENiagaraWindTurbulenceMode=::NewEnumerator0, Wind Speed=<0 0 1>, Wind Speed Scale=200)
   &CurlNoiseForce(
    Bias Noise=false
    Noise Quality / Cost=::NewEnumerator0
    Noise Strength=&RandomRangeFloat(Minimum=1200, Maximum=2000)
    Pan Noise=false
   )
   &Drag(Drag=&RandomRangeFloat(Minimum=2, Maximum=3), Use Linear Drag=true, Use Rotational Drag=false)
   &SolveForcesAndVelocity(Clamp Velocity=false, Limit Acceleration=false)
   &ScaleMeshSize(Scale Factor=&VectorFromFloat(Value=&FloatFromCurve(FloatCurve=DataInterface() {propertyValues {Curve {
         defaultValue=1
         keys [
          {arriveTangent=1.4259, interpMode=RCIM_Cubic, leaveTangent=1.4259, tangentMode=RCTM_User, time=0, value=0.3}
          {arriveTangent=0.1983, interpMode=RCIM_Cubic, leaveTangent=0.1983, tangentMode=RCTM_User, time=1, value=1}
         ]
        }}})))
  ]
  rendererClasses [&NiagaraMeshRendererProperties]
  renderers [{rendererClass=&NiagaraMeshRendererProperties, rendererIndex=0}]
 }
 RoofTiles {
  simTarget=GPUComputeSim
  emitterUpdateScript [
   &EmitterState()
   &SpawnBurst_Instantaneous(Spawn Count=44, Spawn Group=1, Spawn Time=0.02, Use Spawn Probability=false, UseLoopCountLimit=false)
   &SpawnRate(
    SpawnRate=&FloatFromCurve(
     FloatCurve=DataInterface() {propertyValues {Curve {keys [<at 0 1 RCIM_Cubic>, {arriveTangent=-4.9583, tangentMode=RCTM_User, time=0.5, value=0}]}}}
     CurveIndex=$Emitter.Age
     Scale Curve=700
    )
    Use Spawn Probability=false
   )
  ]
  particleSpawnScript [
   &PartitionParticles(
    Normalized Index In Partition Includes One=true
    Number Of Partitions=4
    Particles Per Partition=24
    Partition Mode:ENiagara_PartitionMode_v2=::NewEnumerator1
   )
   /SetVariables(Particles.WallDirection:Vector3f=&RotateVectorByQuaternion(
     Vector To Rotate=<0.8 0 0>
     Quaternion=&MakeQuaternion(
      Quaternion From=0
      Axis=<0 0 1>
      Angle=&Multiply_FloatByInt(Float=0.25, Integer=$Particles.PartitionParticles.Partition)
      From=?
      To=?
      Angle Type=::NewEnumerator1
      Normalize Vectors=false
     )
    )) {isSetParameters=true, module=SetVariables_#2}
   &InitializeParticle(
    Lifetime Max=3
    Lifetime Min=1.7
    Lifetime Mode=::NewEnumerator1
    Mesh Scale Mode=::NewEnumerator3
    Position Offset=<0 0 -100>
    UsePositionOffset=true
    Write Material Random=true
   )
   /SetVariables(Particles.Position=&AddVectorToPosition(
     Position=$Particles.Position
     Vector=&Add_Vector(
      A=&Add_Vector(
       A=&Multiply_Vector(A=$Particles.WallDirection, B=$System.UpperFloorSize)
       B=&Multiply_Vector(
        A=&Multiply_VectorByFloat(
         Vector=/Niagara/DynamicInputs/Vectors/CrossProduct(Vector A:Vector3f=$Particles.WallDirection, Vector B:Vector3f=<0 0 1>, Normalize Incoming Vectors:NiagaraBool=true)
         Float=&Lerp_Float(A=-0.75, B=0.75, Alpha=$Particles.PartitionParticles.Normalized Index In Partition)
        )
        B=$System.UpperFloorSize
       )
      )
      B=&Multiply_VectorByFloat(
       Vector=&Add_Vector(A=$Particles.WallDirection, B=<0 0 -0.5>)
       Float=&FloatFromCurve(
        FloatCurve=DataInterface() {propertyValues {Curve {
           defaultValue=0
           keys [{arriveTangent=-3.6565, interpMode=RCIM_Cubic, time=0, value=1}, {arriveTangent=-1.521, tangentMode=RCTM_User, time=1, value=-0.2}]
          }}}
        CurveIndex=$Particles.MaterialRandom
        Scale Curve=600
       )
      )
     )
    )) {isSetParameters=true, module=SetVariables_#3}
   &AddVelocity(
    Constrain To Radius=2000
    Module.Use Falloff Distance=false
    Origin Offset=&RandomRangeVector(Minimum=<-20 -20 -3000>, Maximum=<20 20 -2000>)
    Override Seed=false
    Velocity Mode=::NewEnumerator1
    Velocity Speed=&Multiply_Float(
     A=&Lerp_Float(A=0.25, B=1, Alpha=$Particles.MaterialRandom)
     B=&MaskFloatBySpawnGroup(
      Masked Float=&RandomRangeFloat(Minimum=1500, Maximum=3500)
      Passthrough Float=&RandomRangeFloat(Minimum=600, Maximum=1200)
      Spawn Group Mask=1
     )
    )
   ) {module=AddVelocity001}
   &InitialMeshOrientation(
    Mesh Orientation Mode=::NewEnumerator2
    Orientation Vector=$Particles.WallDirection
    Rotation=&MakeVector(X=0, Y=&RandomRangeFloat(Minimum=0.05, Maximum=0.1), Z=0)
    Use Rotation Vector=true
   )
  ]
  particleUpdateScript [
   &ParticleState()
   &CameraOffset(Camera Offset Amount=&FloatFromCurve(
     FloatCurve=DataInterface() {propertyValues {Curve {defaultValue=1, keys [<at 0 0 RCIM_Cubic>, {tangentMode=RCTM_User, time=1, value=1}]}}}
     Scale Curve=150
    ))
   &ScaleMeshSize(Scale Factor=&VectorFromFloat(Value=&FloatFromCurve(FloatCurve=DataInterface() {propertyValues {Curve {keys [<at 0.8 1 RCIM_Cubic>, {arriveTangent=-7.4853, tangentMode=RCTM_User, time=1, value=0}]}}})))
   &UpdateMeshOrientation(
    Rotation Coordinate Space=::NewEnumerator1
    Rotation Rate=&FloatFromCurve(
     FloatCurve=DataInterface() {propertyValues {Curve {
        defaultValue=1
        keys [
         {arriveTangent=-0.6674, interpMode=RCIM_Cubic, leaveTangent=-0.6674, tangentMode=RCTM_User, time=0, value=1}
         {tangentMode=RCTM_User, time=1, value=0.5}
        ]
       }}}
     CurveIndex=$Particles.Age
     Scale Curve=&RandomRangeFloat(Minimum=-2, Maximum=-1)
    )
    Rotation Vector=&Add_Vector(
     A=&CrossProduct(Vector A:Vector3f=$Particles.WallDirection, Vector B:Vector3f=<0 0 1>, Normalize Incoming Vectors:NiagaraBool=true)
     B=&VectorFromFloat(Value=&RandomRangeFloat(Minimum=-0.4, Maximum=0.4))
    )
   )
   &GravityForce(Gravity=&Multiply_VectorByFloat(
     Vector=<0 0 -960>
     Float=&FloatFromCurve(
      FloatCurve=DataInterface() {propertyValues {Curve {defaultValue=1, keys [{arriveTangent=4.696, interpMode=RCIM_Cubic, time=0, value=0.25}, {tangentMode=RCTM_User, time=0.25, value=1}]}}}
      Scale Curve=&RandomRangeFloat(Minimum=0.8, Maximum=1)
     )
    ))
   &Drag(Drag=0.25, Use Linear Drag=true, Use Rotational Drag=false)
   &SolveForcesAndVelocity(Clamp Velocity=false, Limit Acceleration=false)
  ]
  rendererClasses [&NiagaraMeshRendererProperties]
  renderers [{rendererClass=&NiagaraMeshRendererProperties, rendererIndex=0}]
 }
 WallCrumbles {
  emitterUpdateScript [
   &EmitterState()
   &SpawnBurst_Instantaneous(Spawn Count=12, Spawn Group=1, Use Spawn Probability=false, UseLoopCountLimit=false) {module=SpawnBurst_Instantaneous001}
   &SpawnBurst_Instantaneous(Spawn Count=16, Spawn Time=0.07, Use Spawn Probability=false, UseLoopCountLimit=false)
   &SpawnBurst_Instantaneous(Spawn Count=4, Spawn Group=2, Spawn Time=0.03, Use Spawn Probability=false, UseLoopCountLimit=false) {module=SpawnBurst_Instantaneous002}
  ]
  particleSpawnScript [
   /SetVariables(Particles.VisibilityTag=&MaskIntBySpawnGroup(Masked Int=1, Passthrough Int=0, Spawn Group Mask=2)) {isSetParameters=true}
   &PartitionParticles(
    Normalized Index In Partition Includes One=true
    Number Of Partitions=4
    Particles Per Partition=4
    Partition Mode:ENiagara_PartitionMode_v2=::NewEnumerator1
   )
   /SetVariables(Particles.WallDirection:Vector3f=&MaskVectorBySpawnGroup(
     Masked Vector=<0 0 1>
     Passthrough Vector=&RotateVectorByQuaternion(
      Vector To Rotate=<1 0 0>
      Quaternion=&MakeQuaternion(
       Quaternion From=0
       Axis=<0 0 1>
       Angle=&Multiply_FloatByInt(Float=0.25, Integer=$Particles.PartitionParticles.Partition)
       From=?
       To=?
       Angle Type=::NewEnumerator1
       Normalize Vectors=false
      )
     )
     Spawn Group Mask=2
    )) {isSetParameters=true, module=SetVariables_#2}
   &InitializeParticle(
    Color Maximum=<rgba 0.2805 0.3 0.261 1>
    Color Minimum=<rgba 0.1402 0.15 0.1305 1>
    Color Mode=::NewEnumerator2
    Lifetime Max=3.5
    Lifetime Min=2.5
    Lifetime Mode=::NewEnumerator1
    Mesh Scale Max=<4 5 5>
    Mesh Scale Min=<2 4 4>
    Mesh Scale Mode=::NewEnumerator2
    Position Offset=&MakeVector(X=0, Y=0, Z=&MaskFloatBySpawnGroup(Masked Float=200, Passthrough Float=-600, Spawn Group Mask=1))
    UsePositionOffset=true
    Write Material Random=true
   )
   /SetVariables(Particles.Position=&AddVectorToPosition(
     Position=$Particles.Position
     Vector=&MaskVectorBySpawnGroup(
      Masked Vector=<0 0 500>
      Passthrough Vector=&Add_Vector(
       A=&Multiply_Vector(
        A=$Particles.WallDirection
        B=&Multiply_VectorByFloat(
         Vector=$System.UpperFloorSize
         Float=&MaskFloatBySpawnGroup(Masked Float=0.65, Passthrough Float=1, Spawn Group Mask=1)
        )
       )
       B=&Multiply_Vector(
        A=&Multiply_VectorByFloat(
         Vector=&CrossProduct(Vector A:Vector3f=$Particles.WallDirection, Vector B:Vector3f=<0 0 1>, Normalize Incoming Vectors:NiagaraBool=true)
         Float=&Lerp_Float(A=-0.75, B=0.75, Alpha=$Particles.PartitionParticles.Normalized Index In Partition)
        )
        B=&Multiply_VectorByFloat(
         Vector=$System.UpperFloorSize
         Float=&MaskFloatBySpawnGroup(Masked Float=0.65, Passthrough Float=1, Spawn Group Mask=1)
        )
       )
      )
      Spawn Group Mask=2
     )
    )) {isSetParameters=true, module=SetVariables_#3}
   &ShapeLocation(MaskSpawnGroup=false, Override Seed=false, Sphere Radius=50, UseOwnerScale=true)
   &AddVelocity(
    Velocity=&Add_Vector(A=$Particles.WallDirection, B=&RandomRangeVector(Minimum=<-0.1 -0.1 -0.1>, Maximum=<0.1 0.1 0.1>))
    Velocity Speed Scale=&RandomRangeFloat(Minimum=50, Maximum=100)
   )
   &AddVelocity(
    Module.Use Falloff Distance=false
    Origin Offset=<0 0 -1000>
    Override Seed=false
    Velocity Mode=::NewEnumerator1
    Velocity Speed=&Multiply_Float(
     A=&RandomRangeFloat(Minimum=0.4, Maximum=1)
     B=&MaskFloatBySpawnGroup(Masked Float=1500, Passthrough Float=700, Spawn Group Mask=1)
    )
   ) {module=AddVelocity001}
   &InitialMeshOrientation(
    Mesh Orientation Mode=::NewEnumerator2
    Orientation Vector=&Add_Vector(
     A=$Particles.WallDirection
     B=&Multiply_VectorByFloat(
      Vector=&NormalizeVector(Vector To Normalize=$Particles.Velocity, Safe Normalize=true, Safe Fallback Threshold=0, Safe Fallback Vector=<0 0 0.001>)
      Float=2
     )
    )
    Rotation=&MaskVectorBySpawnGroup(
     Masked Vector=&RandomRangeVector(Minimum=<0 0 0>, Maximum=<1 0 0>)
     Passthrough Vector=<0 0 0>
     Spawn Group Mask=2
    )
    Use Rotation Vector=true
   )
  ]
  particleUpdateScript [
   &ParticleState()
   &CameraOffset(Camera Offset Amount=&FloatFromCurve(
     FloatCurve=DataInterface() {propertyValues {Curve {defaultValue=1, keys [<at 0 0 RCIM_Cubic>, {tangentMode=RCTM_User, time=0.5, value=1}]}}}
     Scale Curve=150
    ))
   &ScaleMeshSize(Scale Factor=&Add_Vector(
     A=<1 1 1>
     B=&VectorFromCurve(
      VectorCurve=DataInterface() {propertyValues {
        XCurve {keys [
          {arriveTangent=2.6878, interpMode=RCIM_Cubic, leaveTangent=2.6878, tangentMode=RCTM_User, time=0, value=0}
          {arriveTangent=0.4081, tangentMode=RCTM_User, time=1, value=1}
         ]}
        YCurve {keys [{arriveTangent=2.2456, interpMode=RCIM_Cubic, leaveTangent=2.2456, tangentMode=RCTM_User, time=0, value=0}, <at 1 1>]}
        ZCurve {keys [
          {arriveTangent=0.6534, interpMode=RCIM_Cubic, leaveTangent=0.6534, tangentMode=RCTM_User, time=0, value=0}
          {arriveTangent=0.5749, tangentMode=RCTM_User, time=1, value=1}
         ]}
       }}
      Scale Curve=&RandomRangeVector(Minimum=<10 1 3>, Maximum=<40 3 6>)
     )
    ))
   &UpdateMeshOrientation(
    Rotation Coordinate Space=::NewEnumerator1
    Rotation Rate=&MaskFloatBySpawnGroup(Masked Float=0, Passthrough Float=&RandomRangeFloat(Minimum=-0.08, Maximum=-0.04), Spawn Group Mask=2)
    Rotation Vector=&Add_Vector(
     A=&CrossProduct(Vector A:Vector3f=$Particles.WallDirection, Vector B:Vector3f=<0 0 1>, Normalize Incoming Vectors:NiagaraBool=true)
     B=&VectorFromFloat(Value=&RandomRangeFloat(Minimum=-0.2, Maximum=0.2))
    )
   )
   &DynamicMaterialParameters(
    Index 0 Param 1=&FloatFromCurve(FloatCurve=DataInterface() {propertyValues {Curve {
        defaultValue=1
        keys [
         {arriveTangent=32.5365, interpMode=RCIM_Cubic, leaveTangent=32.5365, tangentMode=RCTM_User, time=0, value=0}
         {arriveTangentWeight=0.1491, interpMode=RCIM_Cubic, leaveTangentWeight=0.2236, tangentMode=RCTM_User, time=0.05, value=1}
         {interpMode=RCIM_Cubic, tangentMode=RCTM_User, time=0.6, value=1}
         {arriveTangent=-4.1553, tangentMode=RCTM_User, time=1, value=0}
        ]
       }}})
    Index 0 Param 2=&Multiply_Float(
     A=&RandomRangeFloat(Minimum=1, Maximum=3)
     B=&MaskFloatBySpawnGroup(Masked Float=2, Passthrough Float=1, Spawn Group Mask=2)
    )
    Index 0 Param 3=&FloatFromCurve(
     FloatCurve=DataInterface() {propertyValues {Curve {keys [
         {arriveTangent=-2.3575, interpMode=RCIM_Cubic, leaveTangent=-2.3575, tangentMode=RCTM_User, time=0, value=1}
         {arriveTangent=-0.3909, tangentMode=RCTM_User, time=1, value=0}
        ]}}}
     Scale Curve=&RandomRangeFloat(Minimum=6, Maximum=12)
    )
    Module.Index0Float0Write=true
    Module.Index0Float1Write=true
    Module.Index0Float2Write=true
    Module.Index0Float3Write=false
   )
   &GravityForce(Gravity=&RandomRangeVector(Minimum=<0 0 -1200>, Maximum=<0 0 -600>))
   &SolveForcesAndVelocity(Clamp Velocity=false, Limit Acceleration=false)
  ]
  rendererClasses [&NiagaraMeshRendererProperties]
  renderers [
   {rendererClass=&NiagaraMeshRendererProperties, rendererIndex=0}
   {rendererClass=&NiagaraMeshRendererProperties, rendererIndex=1}
  ]
 }
 WoodFrames {
  emitterUpdateScript [
   &EmitterState()
   &SpawnBurst_Instantaneous(Spawn Count=16, Spawn Group=1, Use Spawn Probability=false, UseLoopCountLimit=false) {module=SpawnBurst_Instantaneous001}
   &SpawnBurst_Instantaneous(Spawn Count=16, Spawn Time=0.09, Use Spawn Probability=false, UseLoopCountLimit=false)
   &SpawnBurst_Instantaneous(Spawn Count=32, Spawn Group=2, Spawn Time=0.02, Use Spawn Probability=false, UseLoopCountLimit=false) {module=SpawnBurst_Instantaneous002}
   &SpawnBurst_Instantaneous(Spawn Count=8, Spawn Group=3, Use Spawn Probability=false, UseLoopCountLimit=false) {module=SpawnBurst_Instantaneous003}
  ]
  particleSpawnScript [
   /SetVariables(Particles.WoodBeams:NiagaraBool=&SetBoolByIntComparison(A=&SpawnGroup(), B=1, Comparison Type=::NewEnumerator0)) {isSetParameters=true}
   /SetVariables(Particles.VisibilityTag=&MakeCustomIntFromBool(Bool=$Particles.WoodBeams, True INT=1, False INT=0)) {isSetParameters=true}
   /SetVariables(Particles.RandomForce:NiagaraFloat=&RandomRangeFloat(Minimum=0.2, Maximum=1)) {isSetParameters=true}
   &PartitionParticles(
    Normalized Index In Partition Includes One=true
    Number Of Partitions=4
    Particles Per Partition=4
    Partition Mode:ENiagara_PartitionMode_v2=::NewEnumerator1
   )
   /SetVariables(Particles.WallDirection:Vector3f=&MaskVectorBySpawnGroup(
     Masked Vector=<0 0 1>
     Passthrough Vector=&RotateVectorByQuaternion(
      Vector To Rotate=<1 0 0>
      Quaternion=&MakeQuaternion(
       Quaternion From=0
       Axis=<0 0 1>
       Angle=&Multiply_FloatByInt(Float=0.25, Integer=$Particles.PartitionParticles.Partition)
       From=?
       To=?
       Angle Type=::NewEnumerator1
       Normalize Vectors=false
      )
     )
     Spawn Group Mask=3
    )) {isSetParameters=true}
   &InitializeParticle(
    Color Mode=::NewEnumerator0
    Lifetime=&RandomRangeFloat(Minimum=2.5, Maximum=4.5)
    Mesh Renderer Array Visibility Mode=::NewEnumerator1
    Mesh Renderer Info=DataInterface() {propertyValues {MeshRenderer=&self:WoodFrames_0.NiagaraMeshRendererProperties_1}}
    Mesh Renderer Mesh Index:NiagaraInt32=&SelectIntFromArray(
     Array Sampling Mode=::NewEnumerator1
     Direct Array Index=?
     Int Selection Array=DataInterface() {propertyValues {IntData [0,0,0,0,1]}}
     Random Seed=0
    )
    Mesh Scale Max=<1 1.7 1.7>
    Mesh Scale Min=<1 1.25 1.25>
    Mesh Scale Mode=::NewEnumerator2
    Position Offset=&MakeVector(
     X=0
     Y=0
     Z=&SelectFloatFromArray(
      Direct Array Index=&SpawnGroup()
      Float Selection Array=DataInterface() {propertyValues {FloatData [-600,200,-200]}}
     )
    )
    UsePositionOffset=true
    Write Material Random=true
   )
   /SetVariables(Particles.Position=&AddVectorToPosition(
     Position=$Particles.Position
     Vector=&MaskVectorBySpawnGroup(
      Masked Vector=<0 0 1500>
      Passthrough Vector=&Add_Vector(
       A=&Multiply_Vector(
        A=$Particles.WallDirection
        B=&Multiply_VectorByFloat(Vector=$System.UpperFloorSize, Float=&MaskFloatBySpawnGroup(Masked Float=1, Passthrough Float=0.65))
       )
       B=&Multiply_Vector(
        A=&Multiply_VectorByFloat(
         Vector=&CrossProduct(Vector A:Vector3f=$Particles.WallDirection, Vector B:Vector3f=<0 0 1>, Normalize Incoming Vectors:NiagaraBool=true)
         Float=&Lerp_Float(
          A=-0.75
          B=0.75
          Alpha=&Add_Float(
           A=$Particles.PartitionParticles.Normalized Index In Partition
           B=&RandomRangeFloat(Minimum=-0.05, Maximum=0.05)
          )
         )
        )
        B=&Multiply_VectorByFloat(Vector=$System.UpperFloorSize, Float=&MaskFloatBySpawnGroup(Masked Float=1, Passthrough Float=0.65))
       )
      )
      Spawn Group Mask=3
     )
    )) {isSetParameters=true}
   &ShapeLocation(
    MaskSpawnGroup=false
    Override Seed=false
    Sphere Radius=&MaskFloatBySpawnGroup(Masked Float=200, Passthrough Float=50, Spawn Group Mask=3)
    Sphere Surface Distribution=0.5
    UseOwnerScale=true
   )
   &AddVelocity(
    Velocity=&Add_Vector(
     A=$Particles.WallDirection
     B=&RandomVector(Vector Scale=&RandomRangeFloat(Minimum=0.1, Maximum=0.2))
    )
    Velocity Speed Scale=&RandomRangeFloat(Minimum=200, Maximum=600)
   )
   &AddVelocity(
    Module.Use Falloff Distance=false
    Origin Offset=<0 0 -1000>
    Override Seed=false
    Velocity Mode=::NewEnumerator1
    Velocity Speed=&Multiply_Float(
     A=$Particles.RandomForce
     B=&SelectFloatFromArray(
      Direct Array Index=&SpawnGroup()
      Float Selection Array=DataInterface() {propertyValues {FloatData [600,1600,3000,6000]}}
     )
    )
   ) {module=AddVelocity001}
   &InitialMeshOrientation(
    Mesh Orientation Mode=::NewEnumerator2
    Orientation Vector=$Particles.WallDirection
    Rotation=&MakeVector(
     X=&SelectFloatFromArray(
      Array Sampling Mode=::NewEnumerator0
      Direct Array Index=?
      Float Selection Array=DataInterface() {propertyValues {FloatData [0,0.25,0.5,0.75]}}
      Random Seed=0
     )
     Y=0
     Z=0
    )
    Use Rotation Vector=true
   )
  ]
  particleUpdateScript [
   &ParticleState()
   &CameraOffset(Camera Offset Amount=&FloatFromCurve(
     FloatCurve=DataInterface() {propertyValues {Curve {defaultValue=1, keys [<at 0 0 RCIM_Cubic>, {tangentMode=RCTM_User, time=1, value=1}]}}}
     Scale Curve=70
    ))
   &UpdateMeshOrientation(
    Rotation Coordinate Space=::NewEnumerator1
    Rotation Rate=&FloatFromCurve(
     FloatCurve=DataInterface() {propertyValues {Curve {
        defaultValue=1
        keys [
         {arriveTangent=-0.9725, interpMode=RCIM_Cubic, leaveTangent=-0.9725, tangentMode=RCTM_User, time=0, value=1}
         {arriveTangent=-0.5583, interpMode=RCIM_Cubic, leaveTangent=-0.5583, tangentMode=RCTM_User, time=1, value=0.2}
        ]
       }}}
     CurveIndex=$Particles.Age
     Scale Curve=&Multiply_Float(A=$Particles.RandomForce, B=-2)
    )
    Rotation Vector=&Add_Vector(
     A=&CrossProduct(Vector A:Vector3f=$Particles.WallDirection, Vector B:Vector3f=<0 0 1>, Normalize Incoming Vectors:NiagaraBool=true)
     B=&RandomVector(Vector Scale=&MakeCustomFloatFromBool(
       Bool=$Particles.WoodBeams
       True Float=&RandomRangeFloat(Minimum=0.5, Maximum=1)
       False Float=&RandomRangeFloat(Minimum=0.25, Maximum=0.5)
      ))
    )
   )
   &ScaleMeshSize(Scale Factor=&VectorFromFloat(Value=&FloatFromCurve(
      FloatCurve=DataInterface() {propertyValues {Curve {keys [<at 0.8 1 RCIM_Cubic>, {arriveTangent=-7.4853, tangentMode=RCTM_User, time=1, value=0}]}}}
      Scale Curve=&MaskFloatBySpawnGroup(Masked Float=1.5, Passthrough Float=1, Spawn Group Mask=3)
     )))
   &Drag(Drag=0.5, Use Linear Drag=true, Use Rotational Drag=false)
   &GravityForce(Gravity=&RandomRangeVector(Minimum=<0 0 -1200>, Maximum=<0 0 -960>))
   &SolveForcesAndVelocity(Clamp Velocity=false, Limit Acceleration=false)
  ]
  rendererClasses [&NiagaraMeshRendererProperties]
  renderers [
   {rendererClass=&NiagaraMeshRendererProperties, rendererIndex=0}
   {rendererClass=&NiagaraMeshRendererProperties, rendererIndex=1}
  ]
 }
}
summary {
 systemName=NS_PistolCanary_Castle_TopBurst
 emitters [
  {emitterName=WallCrumbles, rendererClasses [&NiagaraMeshRendererProperties]}
  {emitterName=WoodFrames, rendererClasses [&NiagaraMeshRendererProperties]}
  {emitterName=MeshSmoke, rendererClasses [&NiagaraMeshRendererProperties]}
  {emitterName=LingerSmoke, rendererClasses [&NiagaraSpriteRendererProperties]}
  {emitterName=RoofTiles, simTarget=GPUComputeSim, rendererClasses [&NiagaraMeshRendererProperties]}
  {emitterName=BigMasks, rendererClasses [&NiagaraMeshRendererProperties]}
  {emitterName=GlowBurst, rendererClasses [&NiagaraSpriteRendererProperties, &NiagaraLightRendererProperties]}
  {emitterName=LowUnlitSmoke, rendererClasses [&NiagaraSpriteRendererProperties]}
 ]
 userVariables [
  User.ScaleMult {type=NiagaraFloat, defaultValue=0.5, description=""}
  User.SpawnRatMult {type=NiagaraFloat, defaultValue=1, description=""}
 ]
}
