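# Niagara debris effect for a crumbling tower: roof tiles, wall chunks, wood frames and beams, plus mesh and sprite smoke.
# Each emitter partitions particles across four walls, rotating a wall direction in quarter turns and spreading spawns along edges sized by System.UpperFloorSize.
# Debris is thrown outward under gravity, with staggered bursts for walls and wood, tumbling tiles and wood, and GPU-simulated tiles.
# Smoke spawn height drops with emitter age, and curve-driven dynamic material parameters animate the smoke.
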NiagaraSystem /PistolCanary/Assets/FX/Debris/NS_PistolCanary_Tower_Crumble
dependencies {
 usedDataInterfaces [
  &NiagaraDataInterfaceCurve
  &NiagaraDataInterfaceVectorCurve
  &NiagaraDataInterfaceArrayFloat
  &NiagaraDataInterfaceColorCurve
 ]
 usedDynamicInputs [&ReturnNormalizedExecIndex]
 usedModules [
  &self:SystemSpawnScript.SystemScriptSource.SystemScriptGraph.NiagaraNodeAssignment_1.SetVariables_#1
  &SystemState
 ]
}
emitters {
 HighMeshSmoke {
  emitterSpawnScript [
   /SetVariables(Emitter.Global Color:LinearColor=<rgba 0 0 0 1>, Emitter.Global Scale=3) {isSetParameters=true}
   /SetVariables(
    Emitter.Color2:LinearColor=&MultiplyLinearColor(Linear Color=$Emitter.Global Color, Scale Factor=<2 1 1 1>)
    Emitter.Depth Fade:NiagaraFloat=&Multiply_Float(A=50, B=$Emitter.Global Scale)
    Emitter.Dissolve Noise Power=1.1
    Emitter.Dissolve Noise Scale:LinearColor=&MultiplyLinearColor(
     Linear Color=/Niagara/DynamicInputs/TypeConversions/MakeLinearColorFromFloat(Value:NiagaraFloat=&FloatFromCurve(FloatCurve=DataInterface() {propertyValues {Curve {keys [<at 0 2>,<at 1 1>,<at 2 0.5>]}}}, CurveIndex=$Emitter.Global Scale))
     Scale Factor=<0.0008 0.0008 0.0009 1>
    )
    Emitter.IS_SimpleSmoke=true
    Emitter.Near Camera Fade:NiagaraFloat=&Multiply_Float(A=200, B=$Emitter.Global Scale)
   ) {isSetParameters=true}
   /SetVariables(Emitter.Dissolve Scale Adjsutment:NiagaraFloat=&FloatFromCurve(
     FloatCurve=DataInterface() {propertyValues {Curve {keys [<at 0 3>,<at 1 1>,<at 5 0.5>,<at 20 0.4>,<at 100 0.4>]}}}
     CurveIndex=$Emitter.Global Scale
     Scale Curve=$Emitter.Global Scale
    )) {isSetParameters=true}
  ]
  emitterUpdateScript [
   &EmitterState()
   &SpawnBurst_Instantaneous(Spawn Count=12, Use Spawn Probability=false, UseLoopCountLimit=false)
   &SpawnRate(SpawnRate=20, Use Spawn Probability=false)
  ]
  particleSpawnScript [
   &PartitionParticles(
    Normalized Index In Partition Includes One=true
    Number Of Partitions=4
    Particles Per Partition=3
    Partition Mode:ENiagara_PartitionMode_v2=::NewEnumerator1
   )
   /SetVariables(Particles.WallDirection:Vector3f=&RotateVectorByQuaternion(
     Vector To Rotate=<1 0 0>
     Quaternion=&MakeQuaternion(
      Quaternion From=0
      Axis=<0 0 1>
      Angle=&Multiply_FloatByInt(Float=0.25, Integer=$Particles.PartitionParticles.Partition)
      From=?
      To=?
      Angle Type=::NewEnumerator1
      Normalize Vectors=false
     )
    )) {isSetParameters=true, module=SetVariables_#2}
   &InitializeParticle(
    Color=<rgba 0.1099 0.079 0.12 1.25>
    Lifetime=&RandomRangeFloat(Minimum=1.5, Maximum=2.5)
    Mesh Scale=&RandomRangeVector(Minimum=<2.5 2 2>, Maximum=<3 3.5 3.5>)
    Mesh Scale Mode=::NewEnumerator4
    Position Offset=&MakeVector(
     X=0
     Y=0
     Z=&Lerp_Float(
      A=-100
      B=-2000
      Alpha=&FloatFromCurve(
       FloatCurve=DataInterface() {propertyValues {Curve {defaultValue=1, keys [<at 0 0 RCIM_Cubic>, {arriveTangent=0.8401, tangentMode=RCTM_User, time=2, value=1}]}}}
       CurveIndex=$Emitter.Age
      )
     )
    )
    UsePositionOffset=true
    Write Material Random=true
   )
   /SetVariables(Particles.Position=&AddVectorToPosition(
     Position=$Particles.Position
     Vector=&Add_Vector(
      A=&Multiply_Vector(
       A=$Particles.WallDirection
       B=&Multiply_VectorByFloat(Vector=$System.UpperFloorSize, Float=&RandomRangeFloat(Minimum=0.5, Maximum=1))
      )
      B=&Multiply_Vector(
       A=&Multiply_VectorByFloat(
        Vector=/Niagara/DynamicInputs/Vectors/CrossProduct(Vector A:Vector3f=$Particles.WallDirection, Vector B:Vector3f=<0 0 1>, Normalize Incoming Vectors:NiagaraBool=true)
        Float=&Lerp_Float(A=-0.8, B=0.8, Alpha=$Particles.PartitionParticles.Normalized Index In Partition)
       )
       B=&Multiply_VectorByFloat(Vector=$System.UpperFloorSize, Float=&RandomRangeFloat(Minimum=0.5, Maximum=1))
      )
     )
    )) {isSetParameters=true, module=SetVariables_#3}
   &ShapeLocation(MaskSpawnGroup=false, Non Uniform Scale=<1 1 2>, Override Seed=false, Sphere Surface Distribution=0.5, UseOwnerScale=true)
   &AddVelocity(
    Velocity=&Add_Vector(
     A=$Particles.WallDirection
     B=&RandomVector(Vector Scale=&RandomRangeFloat(Minimum=0.1, Maximum=0.2))
    )
    Velocity Speed Scale=&RandomRangeFloat(Minimum=250, Maximum=400)
   )
   &CameraOffset(Camera Offset Amount=300)
   /SetVariables(Particles.Scale=&Multiply_VectorByFloat(Vector=$Particles.Scale, Float=$Emitter.Global Scale)) {isSetParameters=true}
  ]
  particleUpdateScript [
   &ParticleState()
   &Drag(Drag=&RandomRangeFloat(Minimum=0.5, Maximum=0.7), Use Linear Drag=true, Use Rotational Drag=false)
   &DynamicMaterialParameters(
    Index 0 Param 1=&Multiply_Float(A=-100, B=$Emitter.Global Scale)
    Index 0 Param 2=&FloatFromCurve(FloatCurve=DataInterface() {propertyValues {Curve {keys [
         {tangentMode=RCTM_User, time=0, value=1}
         {arriveTangent=-0.0405, interpMode=RCIM_Cubic, leaveTangent=-0.0405, tangentMode=RCTM_User, time=0.4, value=1}
         {arriveTangent=-1.6592, interpMode=RCIM_Cubic, leaveTangent=-1.6592, tangentMode=RCTM_User, time=1, value=0.4}
        ]}}})
    Index 0 Param 3=&FloatFromCurve(
     FloatCurve=DataInterface() {propertyValues {Curve {keys [
         {arriveTangent=1.4725, interpMode=RCIM_Cubic, leaveTangent=1.4725, tangentMode=RCTM_User, time=0, value=0}
         {arriveTangent=0.4975, tangentMode=RCTM_User, time=1, value=1}
        ]}}}
     Scale Curve=&Multiply_Float(
      A=&RandomRangeFloat(Minimum=100, Maximum=150)
      B=&Multiply_Float(
       A=$Emitter.Global Scale
       B=&FloatFromCurve(FloatCurve=DataInterface() {propertyValues {Curve {keys [<at 0 1>,<at 100 0.9>,<at 300 0>]}}}, CurveIndex=$Emitter.Global Scale)
      )
     )
    )
    Index 0 Param 4=&FloatFromCurve(
     FloatCurve=DataInterface() {propertyValues {Curve {keys [
         {arriveTangent=0.1175, interpMode=RCIM_Cubic, leaveTangent=1.7315, tangentMode=RCTM_Break, time=0.1, value=-3}
         {arriveTangent=14.4979, interpMode=RCIM_Cubic, leaveTangent=14.4841, tangentMode=RCTM_Break, time=1, value=6}
        ]}}}
     Scale Curve=0.5
    )
    Index 1 Param 1=&FloatFromCurve(
     FloatCurve=DataInterface() {propertyValues {Curve {keys [
         {arriveTangent=-31.7902, interpMode=RCIM_Cubic, time=0, value=5}
         {arriveTangent=4.4473, interpMode=RCIM_Cubic, leaveTangent=4.4473, tangentMode=RCTM_User, time=0.15, value=0.7}
         {interpMode=RCIM_Cubic, leaveTangent=-0.2867, time=0.4, value=7}
        ]}}}
     Scale Curve=&RandomRangeFloat(Minimum=5, Maximum=6)
    )
    Index 1 Param 2=&Divide_Float(
     A=&RandomRangeFloat(Minimum=0.017, Maximum=0.02)
     B=&Multiply_Float(
      A=$Emitter.Global Scale
      B=&FloatFromCurve(
       FloatCurve=DataInterface() {propertyValues {Curve {keys [
           {arriveTangent=-0.812, interpMode=RCIM_Cubic, leaveTangent=-0.812, tangentMode=RCTM_User, time=0, value=1.5}
           <at 1 1>
           <at 2 0.82>
           <at 100 0.75>
           <at 1000 1>
          ]}}}
       CurveIndex=$Emitter.Global Scale
      )
     )
    )
    Index 1 Param 3=&FloatFromCurve(
     FloatCurve=DataInterface() {propertyValues {Curve {keys [{interpMode=RCIM_Cubic, tangentMode=RCTM_User, time=0.4, value=0.5}, <at 0.8 0.1>]}}}
     Scale Curve=5
    )
    Index 1 Param 4=1
    Index 2 Param 1=&FloatFromCurve(
     FloatCurve=DataInterface() {propertyValues {Curve {keys [
         {arriveTangent=0.2023, interpMode=RCIM_Cubic, leaveTangent=0.2023, tangentMode=RCTM_User, time=0.3, value=0.1}
         {arriveTangent=0.5029, interpMode=RCIM_Cubic, leaveTangent=0.5029, tangentMode=RCTM_User, time=1, value=0.3}
        ]}}}
     Scale Curve=&Multiply_Float(
      A=&Multiply_Float(
       A=&FloatFromCurve(
        FloatCurve=DataInterface() {propertyValues {Curve {keys [
            {arriveTangent=1.5513, interpMode=RCIM_Cubic, leaveTangent=1.5513, tangentMode=RCTM_User, time=0.1, value=0}
            {arriveTangent=-6.1361, interpMode=RCIM_Cubic, leaveTangent=-6.1361, tangentMode=RCTM_User, time=1, value=3}
            <at 2 0.3>
            <at 10 0.1>
            <at 20 0.02>
            <at 50 0.002>
            <at 100 0.001>
            <at 200 0>
           ]}}}
        CurveIndex=$Emitter.Global Scale
       )
       B=$Emitter.Global Scale
      )
      B=$Emitter.Global Scale
     )
    )
    Index 2 Param 2=&FloatFromCurve(
     FloatCurve=DataInterface() {propertyValues {Curve {keys [
         {interpMode=RCIM_Cubic, tangentMode=RCTM_User, time=0.3, value=0}
         {arriveTangent=2.4333, interpMode=RCIM_Cubic, leaveTangent=2.4333, tangentMode=RCTM_User, time=1, value=1}
        ]}}}
     Scale Curve=0.15
    )
    Index 2 Param 3=&FloatFromCurve(
     FloatCurve=DataInterface() {propertyValues {Curve {keys [
         {arriveTangent=9.8846, interpMode=RCIM_Cubic, leaveTangent=9.8846, tangentMode=RCTM_User, time=0, value=0}
         <at 0.2 1 RCIM_Cubic>
         {interpMode=RCIM_Cubic, leaveTangent=-0.7527, time=0.6, value=0}
        ]}}}
     Scale Curve=200
    )
    Index 2 Param 4=&FloatFromCurve(
     FloatCurve=DataInterface() {propertyValues {Curve {keys [
         {arriveTangent=-16.6453, interpMode=RCIM_Cubic, leaveTangent=-16.6453, tangentMode=RCTM_User, time=0, value=64}
         {arriveTangent=-87.1792, interpMode=RCIM_Cubic, leaveTangent=-87.1792, tangentMode=RCTM_User, time=0.1, value=24}
         {arriveTangent=-52.5, interpMode=RCIM_Cubic, leaveTangent=-52.5, time=0.2, value=12}
         {arriveTangent=-7.0531, interpMode=RCIM_Cubic, leaveTangent=-7.0531, tangentMode=RCTM_User, time=0.5, value=3}
         {arriveTangent=-0.6909, interpMode=RCIM_Cubic, leaveTangent=-0.6909, tangentMode=RCTM_User, time=1, value=2}
         <at 2 1>
         <at 3 0.3>
         <at 10 0.1>
         <at 20 0.05>
         <at 50 0.02>
         <at 100 0.01>
        ]}}}
     CurveIndex=$Emitter.Global Scale
    )
    Index 3 Param 1=&FloatFromCurve(
     FloatCurve=DataInterface() {propertyValues {Curve {keys [
         {arriveTangent=2.2353, interpMode=RCIM_Cubic, leaveTangent=2.2353, tangentMode=RCTM_User, time=0, value=0.2}
         {arriveTangent=1.4416, interpMode=RCIM_Cubic, leaveTangent=1.4416, tangentMode=RCTM_User, time=0.2, value=0.6}
         {interpMode=RCIM_Cubic, tangentMode=RCTM_User, time=0.8, value=1.25}
        ]}}}
     Scale Curve=-2
    )
    Index 3 Param 2=&FloatFromCurve(
     FloatCurve=DataInterface() {propertyValues {Curve {keys [
         {arriveTangent=5.1563, interpMode=RCIM_Cubic, leaveTangent=5.1563, tangentMode=RCTM_User, time=0, value=0}
         {arriveTangent=1.4416, interpMode=RCIM_Cubic, leaveTangent=1.4416, tangentMode=RCTM_User, time=0.1699, value=0.4}
         {interpMode=RCIM_Cubic, tangentMode=RCTM_User, time=1, value=1.1}
        ]}}}
     Scale Curve=-0.15
    )
    Index 3 Param 3=&FloatFromCurve(FloatCurve=DataInterface() {propertyValues {Curve {keys [
         {arriveTangent=-11.5239, interpMode=RCIM_Cubic, leaveTangent=-11.5239, tangentMode=RCTM_User, time=0.04, value=65}
         {arriveTangent=-195.7668, interpMode=RCIM_Cubic, leaveTangent=-195.7668, tangentMode=RCTM_User, time=0.18, value=50}
         {arriveTangent=-43.7054, interpMode=RCIM_Cubic, leaveTangent=-43.7054, tangentMode=RCTM_User, time=0.35, value=2}
         {interpMode=RCIM_Cubic, tangentMode=RCTM_User, time=0.5, value=0}
        ]}}})
    Index 3 Param 4=&FloatFromCurve(
     FloatCurve=DataInterface() {propertyValues {Curve {keys [{tangentMode=RCTM_User, time=0, value=0}, {tangentMode=RCTM_User, time=1, value=1}]}}}
     Scale Curve=-700
    )
    Module.Index0Param0Write=true
    Module.Index0Param1Write=true
    Module.Index0Param2Write=true
    Module.Index0Param3Write=true
    Module.Index1Param0Write=true
    Module.Index1Param1Write=true
    Module.Index1Param2Write=true
    Module.Index1Param3Write=true
    Module.Index2Param0Write=true
    Module.Index2Param1Write=true
    Module.Index2Param2Write=true
    Module.Index2Param3Write=true
    Module.Index3Param0Write=true
    Module.Index3Param1Write=true
    Module.Index3Param2Write=true
    Module.Index3Param3Write=true
    Write Parameter Index 1=true
    Write Parameter Index 2=true
    Write Parameter Index 3=true
   ) {module=DynamicMaterialParameters001}
   &ScaleColor(
    Linear Color Curve=DataInterface() {propertyValues {
      AlphaCurve {keys [
        {arriveTangent=13.8652, interpMode=RCIM_Cubic, time=0, value=0}
        <at 0.2 1>
        {arriveTangentWeight=0.3303, interpMode=RCIM_Cubic, leaveTangentWeight=0.1386, tangentMode=RCTM_User, time=0.5, value=1}
        {arriveTangent=-4.5785, tangentMode=RCTM_User, time=0.9894, value=0}
       ]}
      BlueCurve {keys [<at 0.1888 0.8 RCIM_Cubic>,<at 0.6 0.4497>]}
      GreenCurve {keys [<at 0.1888 0.8 RCIM_Cubic>,<at 0.6 0.4497>]}
      RedCurve {keys [<at 0.1888 0.8 RCIM_Cubic>,<at 0.6 0.4497>]}
     }}
    Scale Mode=::NewEnumerator2
   )
   &CurlNoiseForce(
    Bias Noise=false
    Noise Quality / Cost=::NewEnumerator0
    Noise Strength=&RandomRangeFloat(Minimum=1200, Maximum=2000)
    Pan Noise=false
   )
   &SolveForcesAndVelocity(Clamp Velocity=false, Limit Acceleration=false)
   &ScaleMeshSize(Scale Factor=&VectorFromFloat(Value=&FloatFromCurve(FloatCurve=DataInterface() {propertyValues {Curve {
         defaultValue=1
         keys [
          {arriveTangent=1.221, interpMode=RCIM_Cubic, leaveTangent=1.221, tangentMode=RCTM_User, time=0, value=0.4}
          {interpMode=RCIM_Cubic, tangentMode=RCTM_User, time=1, value=1}
         ]
        }}})))
  ]
  rendererClasses [&NiagaraMeshRendererProperties]
  renderers [{rendererClass=&NiagaraMeshRendererProperties, rendererIndex=0}]
 }
 LowUnlitSmoke {
  emitterUpdateScript [
   &EmitterState()
   &SpawnBurst_Instantaneous(Spawn Count=8, Use Spawn Probability=false, UseLoopCountLimit=false)
   &SpawnRate(SpawnRate=10, Use Spawn Probability=false)
  ]
  particleSpawnScript [
   &PartitionParticles(
    Normalized Index In Partition Includes One=true
    Number Of Partitions=4
    Particles Per Partition=3
    Partition Mode:ENiagara_PartitionMode_v2=::NewEnumerator1
   )
   /SetVariables(Particles.WallDirection:Vector3f=&RotateVectorByQuaternion(
     Vector To Rotate=<1 0 0>
     Quaternion=&MakeQuaternion(
      Quaternion From=0
      Axis=<0 0 1>
      Angle=&Multiply_FloatByInt(Float=0.25, Integer=$Particles.PartitionParticles.Partition)
      From=?
      To=?
      Angle Type=::NewEnumerator1
      Normalize Vectors=false
     )
    )) {isSetParameters=true, module=SetVariables_#2}
   &InitializeParticle(
    Color Maximum=<rgba 0.05 0.0278 0.0331 1.5>
    Color Minimum=<rgba 0.04 0.0255 0.0314 1.5>
    Color Mode=::NewEnumerator2
    Lifetime=&RandomRangeFloat(Minimum=1.5, Maximum=2.5)
    Mesh Scale=&RandomRangeVector(Minimum=<2.5 2 2>, Maximum=<3 3.5 3.5>)
    Mesh Scale Mode=::NewEnumerator4
    Position Offset=&MakeVector(
     X=0
     Y=0
     Z=&Lerp_Float(
      A=-200
      B=-2000
      Alpha=&FloatFromCurve(
       FloatCurve=DataInterface() {propertyValues {Curve {defaultValue=1, keys [<at 0 0 RCIM_Cubic>, {arriveTangent=0.468, tangentMode=RCTM_User, time=2, value=1}]}}}
       CurveIndex=$Emitter.Age
      )
     )
    )
    Sprite Rotation Mode=::NewEnumerator2
    Sprite Size Mode=::NewEnumerator1
    Uniform Sprite Size Max=1500
    Uniform Sprite Size Min=900
    UsePositionOffset=true
    Write Material Random=true
   )
   /SetVariables(Particles.Position=&AddVectorToPosition(
     Position=$Particles.Position
     Vector=&Add_Vector(
      A=&Multiply_Vector(
       A=$Particles.WallDirection
       B=&Multiply_VectorByFloat(Vector=$System.UpperFloorSize, Float=&RandomRangeFloat(Minimum=0.5, Maximum=1))
      )
      B=&Multiply_Vector(
       A=&Multiply_VectorByFloat(
        Vector=&CrossProduct(Vector A:Vector3f=$Particles.WallDirection, Vector B:Vector3f=<0 0 1>, Normalize Incoming Vectors:NiagaraBool=true)
        Float=&Lerp_Float(A=-0.9, B=0.9, Alpha=$Particles.PartitionParticles.Normalized Index In Partition)
       )
       B=&Multiply_VectorByFloat(Vector=$System.UpperFloorSize, Float=&RandomRangeFloat(Minimum=0.5, Maximum=1))
      )
     )
    )) {isSetParameters=true, module=SetVariables_#3}
   &ShapeLocation(MaskSpawnGroup=false, Override Seed=false, Sphere Surface Distribution=0.5, UseOwnerScale=true)
   &AddVelocity(
    Velocity=&Add_Vector(
     A=$Particles.WallDirection
     B=&RandomVector(Vector Scale=&RandomRangeFloat(Minimum=0.1, Maximum=0.2))
    )
    Velocity Speed Scale=&RandomRangeFloat(Minimum=250, Maximum=400)
   )
   &CameraOffset(Camera Offset Amount=300)
  ]
  particleUpdateScript [
   &ParticleState()
   &ScaleSpriteSize(Uniform Curve Sprite Scale=DataInterface() {propertyValues {Curve {
       defaultValue=1
       keys [
        {arriveTangent=1.221, interpMode=RCIM_Cubic, leaveTangent=1.221, tangentMode=RCTM_User, time=0, value=0.4}
        {interpMode=RCIM_Cubic, tangentMode=RCTM_User, time=1, value=1}
       ]
      }}})
   &Drag(Drag=&RandomRangeFloat(Minimum=0.5, Maximum=0.7), Use Linear Drag=true, Use Rotational Drag=false)
   &ScaleColor(
    Scale Alpha=&FloatFromCurve(FloatCurve=DataInterface() {propertyValues {Curve {
        defaultValue=1
        keys [
         {arriveTangent=8.3954, interpMode=RCIM_Cubic, leaveTangent=8.3954, tangentMode=RCTM_User, time=0, value=0}
         {
          arriveTangent=-0.5026
          arriveTangentWeight=0.1491
          interpMode=RCIM_Cubic
          leaveTangent=-0.5026
          leaveTangentWeight=0.2236
          tangentMode=RCTM_User
          time=0.2
          value=1.25
         }
         {arriveTangentWeight=0.1, interpMode=RCIM_Cubic, leaveTangentWeight=0.1667, tangentMode=RCTM_User, time=0.75, value=1}
         {arriveTangent=-6.06, tangentMode=RCTM_User, time=1, value=0}
        ]
       }}})
    ScaleA=true
    ScaleRGB=false
   )
   &SolveForcesAndVelocity(Clamp Velocity=false, Limit Acceleration=false)
   &DynamicMaterialParameters(
    Index 0 Param 1=&Add_Float(
     A=$Particles.MaterialRandom
     B=&FloatFromCurve(FloatCurve=DataInterface() {propertyValues {Curve {keys [
          {arriveTangent=2.7533, interpMode=RCIM_Cubic, leaveTangent=1.8958, tangentMode=RCTM_User, time=0, value=1}
          {arriveTangent=0.4122, interpMode=RCIM_Cubic, leaveTangent=0.4122, tangentMode=RCTM_User, time=0.321, value=1.3453}
          {arriveTangent=0.1513, interpMode=RCIM_Cubic, leaveTangent=0.1513, tangentMode=RCTM_User, time=1, value=1.5}
         ]}}})
    )
    Index 0 Param 4=&FloatFromCurve(FloatCurve=DataInterface() {propertyValues {Curve {keys [<at 0.5 1 RCIM_Cubic>,<at 1 0>]}}})
    Index 2 Param 1=&FloatFromCurve(FloatCurve=DataInterface() {propertyValues {Curve {keys [
         {interpMode=RCIM_Cubic, leaveTangent=4.2485, tangentMode=RCTM_Break, time=0, value=0}
         {arriveTangent=1.0548, interpMode=RCIM_Cubic, leaveTangent=1.0548, tangentMode=RCTM_User, time=0.17, value=0.5715}
         {arriveTangent=0.4121, tangentMode=RCTM_User, time=1, value=1}
        ]}}})
    Index 2 Param 2=&FloatFromCurve(
     FloatCurve=DataInterface() {propertyValues {Curve {keys [
         {arriveTangent=0.4242, interpMode=RCIM_Cubic, leaveTangent=0.4242, tangentMode=RCTM_User, time=0, value=0.1}
         {arriveTangent=0.1815, interpMode=RCIM_Cubic, leaveTangent=0.1815, tangentMode=RCTM_User, time=1, value=0.4}
        ]}}}
     Scale Curve=&RandomRangeFloat(Minimum=0.5, Maximum=1)
    )
    Index 2 Param 3=&FloatFromCurve(
     FloatCurve=DataInterface() {propertyValues {Curve {keys [
         {interpMode=RCIM_Cubic, leaveTangent=1.2101, tangentMode=RCTM_Break, time=0, value=0}
         {arriveTangent=0.6182, interpMode=RCIM_Cubic, leaveTangent=11.1233, tangentMode=RCTM_Break, time=1, value=1}
        ]}}}
     Scale Curve=3
    )
    Index 2 Param 4=&FloatFromCurve(
     FloatCurve=DataInterface() {propertyValues {Curve {keys [
         {arriveTangent=-0.5563, interpMode=RCIM_Cubic, leaveTangent=-0.5563, tangentMode=RCTM_User, time=0, value=1}
         {interpMode=RCIM_Cubic, leaveTangent=-1.2956, tangentMode=RCTM_Break, time=1, value=0.75}
        ]}}}
     Scale Curve=1.25
    )
    Module.Index0Float0Write=true
    Module.Index0Float1Write=false
    Module.Index0Float2Write=true
    Module.Index0Float3Write=true
    Module.Index2Float0Write=true
    Module.Index2Float1Write=true
    Module.Index2Float2Write=true
    Module.Index2Float3Write=true
    Write Parameter Index 2=::NewEnumerator1
   )
  ]
  rendererClasses [&NiagaraSpriteRendererProperties]
  renderers [{rendererClass=&NiagaraSpriteRendererProperties}]
 }
 RoofTiles {
  simTarget=GPUComputeSim
  emitterUpdateScript [
   &EmitterState(Life Cycle Mode=::NewEnumerator1, Loop Behavior=::NewEnumerator1, Loop Duration=1.25, UseLoopDelay=false)
   &SpawnBurst_Instantaneous(Spawn Count=22, Use Spawn Probability=false, UseLoopCountLimit=false)
   &SpawnRate(
    SpawnRate=&FloatFromCurve(
     FloatCurve=DataInterface() {propertyValues {Curve {keys [<at 0 1 RCIM_Cubic>, {arriveTangent=-4.9583, tangentMode=RCTM_User, time=0.5, value=0}]}}}
     CurveIndex=$Emitter.Age
     Scale Curve=200
    )
    Use Spawn Probability=false
   )
  ]
  particleSpawnScript [
   &PartitionParticles(
    Normalized Index In Partition Includes One=true
    Number Of Partitions=4
    Particles Per Partition=24
    Partition Mode:ENiagara_PartitionMode_v2=::NewEnumerator1
   )
   /SetVariables(Particles.WallDirection:Vector3f=&RotateVectorByQuaternion(
     Vector To Rotate=<0.8 0 0>
     Quaternion=&MakeQuaternion(
      Quaternion From=0
      Axis=<0 0 1>
      Angle=&Multiply_FloatByInt(Float=0.25, Integer=$Particles.PartitionParticles.Partition)
      From=?
      To=?
      Angle Type=::NewEnumerator1
      Normalize Vectors=false
     )
    )) {isSetParameters=true, module=SetVariables_#4}
   &InitializeParticle(
    Lifetime Max=3
    Lifetime Min=1.7
    Lifetime Mode=::NewEnumerator1
    Mesh Scale Mode=::NewEnumerator3
    Position Offset=<0 0 200>
    UsePositionOffset=true
    Write Material Random=true
   )
   /SetVariables(Particles.Position=&AddVectorToPosition(
     Position=$Particles.Position
     Vector=&Add_Vector(
      A=&Add_Vector(
       A=&Multiply_Vector(A=$Particles.WallDirection, B=$System.UpperFloorSize)
       B=&Multiply_Vector(
        A=&Multiply_VectorByFloat(
         Vector=&CrossProduct(Vector A:Vector3f=$Particles.WallDirection, Vector B:Vector3f=<0 0 1>, Normalize Incoming Vectors:NiagaraBool=true)
         Float=&Lerp_Float(A=-1, B=1, Alpha=$Particles.PartitionParticles.Normalized Index In Partition)
        )
        B=$System.UpperFloorSize
       )
      )
      B=&Multiply_VectorByFloat(
       Vector=&Add_Vector(A=$Particles.WallDirection, B=<0 0 -0.5>)
       Float=&FloatFromCurve(
        FloatCurve=DataInterface() {propertyValues {Curve {
           defaultValue=0
           keys [{arriveTangent=-3.6565, interpMode=RCIM_Cubic, time=0, value=1}, {arriveTangent=-1.8303, tangentMode=RCTM_User, time=1, value=0}]
          }}}
        CurveIndex=&RandomRangeFloat(Minimum=0, Maximum=1)
        Scale Curve=400
       )
      )
     )
    )) {isSetParameters=true, module=SetVariables_#5}
   &AddVelocity(
    Velocity=&Add_Vector(A=$Particles.WallDirection, B=&RandomRangeVector(Minimum=<-0.2 -0.2 -0.2>, Maximum=<0.2 0.2 0.2>))
    Velocity Speed Scale=&RandomRangeFloat(Minimum=300, Maximum=500)
   )
   &InitialMeshOrientation(
    Mesh Orientation Mode=::NewEnumerator2
    Orientation Vector=$Particles.WallDirection
    Rotation=&MakeVector(X=0, Y=&RandomRangeFloat(Minimum=0.05, Maximum=0.1), Z=0)
    Use Rotation Vector=true
   )
  ]
  particleUpdateScript [
   &ParticleState()
   &CameraOffset(Camera Offset Amount=&FloatFromCurve(
     FloatCurve=DataInterface() {propertyValues {Curve {defaultValue=1, keys [<at 0 0 RCIM_Cubic>, {tangentMode=RCTM_User, time=1, value=1}]}}}
     Scale Curve=150
    ))
   &ScaleMeshSize(Scale Factor=&VectorFromFloat(Value=&FloatFromCurve(FloatCurve=DataInterface() {propertyValues {Curve {keys [<at 0.8 1 RCIM_Cubic>, {arriveTangent=-7.4853, tangentMode=RCTM_User, time=1, value=0}]}}})))
   &UpdateMeshOrientation(
    Rotation Coordinate Space=::NewEnumerator1
    Rotation Rate=&FloatFromCurve(
     FloatCurve=DataInterface() {propertyValues {Curve {
        defaultValue=1
        keys [{arriveTangent=1.6528, interpMode=RCIM_Cubic, leaveTangent=1.6528, tangentMode=RCTM_User, time=0, value=0}, <at 1 1 RCIM_Cubic>]
       }}}
     CurveIndex=$Particles.Age
     Scale Curve=&RandomRangeFloat(Minimum=-1.4, Maximum=-0.6)
    )
    Rotation Vector=&Add_Vector(
     A=&CrossProduct(Vector A:Vector3f=$Particles.WallDirection, Vector B:Vector3f=<0 0 1>, Normalize Incoming Vectors:NiagaraBool=true)
     B=&VectorFromFloat(Value=&RandomRangeFloat(Minimum=-0.4, Maximum=0.4))
    )
   )
   &GravityForce(Gravity=&Multiply_VectorByFloat(
     Vector=<0 0 -960>
     Float=&FloatFromCurve(
      FloatCurve=DataInterface() {propertyValues {Curve {
         defaultValue=1
         keys [
          {arriveTangent=3.0253, interpMode=RCIM_Cubic, leaveTangent=3.0253, tangentMode=RCTM_User, time=0, value=0.25}
          {tangentMode=RCTM_User, time=0.25, value=1}
         ]
        }}}
      Scale Curve=&RandomRangeFloat(Minimum=0.7, Maximum=1)
     )
    ))
   &SolveForcesAndVelocity(Clamp Velocity=false, Limit Acceleration=false)
  ]
  rendererClasses [&NiagaraMeshRendererProperties]
  renderers [{rendererClass=&NiagaraMeshRendererProperties, rendererIndex=0}]
 }
 WallCrumbles {
  emitterUpdateScript [
   &EmitterState(Life Cycle Mode=::NewEnumerator1, Loop Behavior=::NewEnumerator1, Loop Duration=1.25, UseLoopDelay=false)
   &SpawnBurst_Instantaneous(Spawn Count=12, Use Spawn Probability=false, UseLoopCountLimit=false)
   &SpawnBurst_Instantaneous(Spawn Count=12, Spawn Group=1, Spawn Time=0.3, Use Spawn Probability=false, UseLoopCountLimit=false) {module=SpawnBurst_Instantaneous001}
  ]
  particleSpawnScript [
   &PartitionParticles(
    Normalized Index In Partition Includes One=true
    Number Of Partitions=4
    Particles Per Partition=3
    Partition Mode:ENiagara_PartitionMode_v2=::NewEnumerator1
   )
   /SetVariables(Particles.WallDirection:Vector3f=&RotateVectorByQuaternion(
     Vector To Rotate=<1 0 0>
     Quaternion=&MakeQuaternion(
      Quaternion From=0
      Axis=<0 0 1>
      Angle=&Multiply_FloatByInt(Float=0.25, Integer=$Particles.PartitionParticles.Partition)
      From=?
      To=?
      Angle Type=::NewEnumerator1
      Normalize Vectors=false
     )
    )) {isSetParameters=true, module=SetVariables_#4}
   &InitializeParticle(
    Color Maximum=<rgba 0.2805 0.3 0.261 1>
    Color Minimum=<rgba 0.1402 0.15 0.1305 1>
    Color Mode=::NewEnumerator2
    Lifetime Max=3
    Lifetime Min=2.5
    Lifetime Mode=::NewEnumerator1
    Mesh Scale Max=<1 5 5>
    Mesh Scale Min=<1 4 4>
    Mesh Scale Mode=::NewEnumerator2
    Position Offset=&MakeVector(X=0, Y=0, Z=&MaskFloatBySpawnGroup(Masked Float=-250, Passthrough Float=-550))
    UsePositionOffset=true
    Write Material Random=true
   )
   /SetVariables(Particles.Position=&AddVectorToPosition(
     Position=$Particles.Position
     Vector=&Add_Vector(
      A=&Multiply_Vector(A=$Particles.WallDirection, B=$System.UpperFloorSize)
      B=&Multiply_Vector(
       A=&Multiply_VectorByFloat(
        Vector=&CrossProduct(Vector A:Vector3f=$Particles.WallDirection, Vector B:Vector3f=<0 0 1>, Normalize Incoming Vectors:NiagaraBool=true)
        Float=&Lerp_Float(A=-0.9, B=0.9, Alpha=$Particles.PartitionParticles.Normalized Index In Partition)
       )
       B=$System.UpperFloorSize
      )
     )
    )) {isSetParameters=true, module=SetVariables_#5}
   &AddVelocity(
    Velocity=&Add_Vector(A=$Particles.WallDirection, B=&RandomRangeVector(Minimum=<-0.1 -0.1 -0.1>, Maximum=<0.1 0.1 0.1>))
    Velocity Speed Scale=&RandomRangeFloat(Minimum=100, Maximum=300)
   )
   &InitialMeshOrientation(Mesh Orientation Mode=::NewEnumerator2, Orientation Vector=$Particles.WallDirection, Use Rotation Vector=false)
  ]
  particleUpdateScript [
   &ParticleState()
   &CameraOffset(Camera Offset Amount=&FloatFromCurve(
     FloatCurve=DataInterface() {propertyValues {Curve {
        defaultValue=1
        keys [
         {arriveTangent=3.7852, interpMode=RCIM_Cubic, leaveTangent=3.7852, tangentMode=RCTM_User, time=0, value=0}
         {tangentMode=RCTM_User, time=0.5, value=1}
        ]
       }}}
     Scale Curve=150
    ))
   &ScaleMeshSize(Scale Factor=&Add_Vector(
     A=<1 1 1>
     B=&VectorFromCurve(
      VectorCurve=DataInterface() {propertyValues {XCurve {keys [<at 0 0>,<at 1 1>]}, YCurve {keys [<at 0 0>,<at 1 1>]}, ZCurve {keys [<at 0 0>,<at 1 1>]}}}
      Scale Curve=&RandomRangeVector(Minimum=<1 0.5 1>, Maximum=<25 2 4>)
     )
    ))
   &DynamicMaterialParameters(
    Index 0 Param 1=&FloatFromCurve(FloatCurve=DataInterface() {propertyValues {Curve {
        defaultValue=1
        keys [
         {arriveTangent=32.5365, interpMode=RCIM_Cubic, leaveTangent=32.5365, tangentMode=RCTM_User, time=0, value=0}
         {arriveTangentWeight=0.1491, interpMode=RCIM_Cubic, leaveTangentWeight=0.2236, tangentMode=RCTM_User, time=0.05, value=1}
         {interpMode=RCIM_Cubic, tangentMode=RCTM_User, time=0.7, value=1}
         {arriveTangent=-4.1553, tangentMode=RCTM_User, time=1, value=0}
        ]
       }}})
    Index 0 Param 2=&RandomRangeFloat(Minimum=1, Maximum=4)
    Index 0 Param 3=&FloatFromCurve(
     FloatCurve=DataInterface() {propertyValues {Curve {keys [
         {arriveTangent=-2.3575, interpMode=RCIM_Cubic, leaveTangent=-2.3575, tangentMode=RCTM_User, time=0, value=1}
         {arriveTangent=-0.3909, tangentMode=RCTM_User, time=1, value=0}
        ]}}}
     Scale Curve=&RandomRangeFloat(Minimum=6, Maximum=12)
    )
    Module.Index0Float0Write=true
    Module.Index0Float1Write=true
    Module.Index0Float2Write=true
    Module.Index0Float3Write=false
   )
   &GravityForce(Gravity=&RandomRangeVector(Minimum=<0 0 -1200>, Maximum=<0 0 -500>))
   &SolveForcesAndVelocity(Clamp Velocity=false, Limit Acceleration=false)
  ]
  rendererClasses [&NiagaraMeshRendererProperties]
  renderers [{rendererClass=&NiagaraMeshRendererProperties, rendererIndex=0}]
 }
 WoodFrames {
  emitterUpdateScript [
   &EmitterState()
   &SpawnBurst_Instantaneous(Spawn Count=12, Use Spawn Probability=false, UseLoopCountLimit=false)
   &SpawnBurst_Instantaneous(Spawn Count=8, Spawn Group=1, Spawn Time=0.31, Use Spawn Probability=false, UseLoopCountLimit=false) {module=SpawnBurst_Instantaneous001}
   &SpawnBurst_Instantaneous(Spawn Count=6, Spawn Group=2, Spawn Time=0.1, Use Spawn Probability=false, UseLoopCountLimit=false) {module=SpawnBurst_Instantaneous002}
   &SpawnBurst_Instantaneous(Spawn Count=12, Spawn Group=3, Spawn Time=0.32, Use Spawn Probability=false, UseLoopCountLimit=false) {module=SpawnBurst_Instantaneous003}
  ]
  particleSpawnScript [
   /SetVariables(Particles.WoodBeams:NiagaraBool=&SetBoolByIntComparison(A=&SpawnGroup(), B=1, Comparison Type=::NewEnumerator0)) {isSetParameters=true}
   /SetVariables(Particles.VisibilityTag=&MakeCustomIntFromBool(Bool=$Particles.WoodBeams, True INT=1, False INT=0)) {isSetParameters=true}
   &PartitionParticles(
    Normalized Index In Partition Includes One=true
    Number Of Partitions=4
    Particles Per Partition=3
    Partition Mode:ENiagara_PartitionMode_v2=::NewEnumerator1
   )
   /SetVariables(Particles.WallDirection:Vector3f=&RotateVectorByQuaternion(
     Vector To Rotate=<1 0 0>
     Quaternion=&MakeQuaternion(
      Quaternion From=0
      Axis=<0 0 1>
      Angle=&Multiply_FloatByInt(Float=0.25, Integer=$Particles.PartitionParticles.Partition)
      From=?
      To=?
      Angle Type=::NewEnumerator1
      Normalize Vectors=false
     )
    )) {isSetParameters=true}
   &InitializeParticle(
    Color Mode=::NewEnumerator0
    Lifetime=&RandomRangeFloat(Minimum=3, Maximum=3.5)
    Mesh Scale Max=<1 1.7 1.7>
    Mesh Scale Min=<1 1.25 1.25>
    Mesh Scale Mode=::NewEnumerator2
    Position Offset=&MakeVector(
     X=0
     Y=0
     Z=&MakeCustomFloatFromBool(
      Bool=$Particles.WoodBeams
      True Float=&MaskFloatBySpawnGroup(Masked Float=-100, Passthrough Float=-350, Spawn Group Mask=2)
      False Float=&MaskFloatBySpawnGroup(Masked Float=-200, Passthrough Float=-500)
     )
    )
    UsePositionOffset=true
    Write Material Random=true
   )
   /SetVariables(Particles.Position=&AddVectorToPosition(
     Position=$Particles.Position
     Vector=&Add_Vector(
      A=&Multiply_Vector(A=$Particles.WallDirection, B=$System.UpperFloorSize)
      B=&Multiply_Vector(
       A=&Multiply_VectorByFloat(
        Vector=&CrossProduct(Vector A:Vector3f=$Particles.WallDirection, Vector B:Vector3f=<0 0 1>, Normalize Incoming Vectors:NiagaraBool=true)
        Float=&Lerp_Float(
         A=-0.9
         B=0.9
         Alpha=&Add_Float(
          A=$Particles.PartitionParticles.Normalized Index In Partition
          B=&RandomRangeFloat(Minimum=-0.05, Maximum=0.05)
         )
        )
       )
       B=$System.UpperFloorSize
      )
     )
    )) {isSetParameters=true}
   &AddVelocity(
    Velocity=&Add_Vector(
     A=$Particles.WallDirection
     B=&RandomVector(Vector Scale=&RandomRangeFloat(Minimum=0.2, Maximum=0.4))
    )
    Velocity Speed Scale=&MakeCustomFloatFromBool(
     Bool=$Particles.WoodBeams
     True Float=&RandomRangeFloat(Minimum=500, Maximum=1000)
     False Float=&RandomRangeFloat(Minimum=150, Maximum=300)
    )
   )
   &InitialMeshOrientation(
    Mesh Orientation Mode=::NewEnumerator2
    Orientation Vector=$Particles.WallDirection
    Rotation=&MakeVector(
     X=&SelectFloatFromArray(
      Array Sampling Mode=::NewEnumerator0
      Direct Array Index=?
      Float Selection Array=DataInterface() {propertyValues {FloatData [0,0.25,0.5,0.75]}}
      Random Seed=0
     )
     Y=0
     Z=0
    )
    Use Rotation Vector=true
   )
  ]
  particleUpdateScript [
   &ParticleState()
   &CameraOffset(Camera Offset Amount=&FloatFromCurve(
     FloatCurve=DataInterface() {propertyValues {Curve {defaultValue=1, keys [<at 0 0 RCIM_Cubic>, {tangentMode=RCTM_User, time=1, value=1}]}}}
     Scale Curve=70
    ))
   &UpdateMeshOrientation(
    Rotation Coordinate Space=::NewEnumerator1
    Rotation Rate=&FloatFromCurve(
     FloatCurve=DataInterface() {propertyValues {Curve {
        defaultValue=1
        keys [{arriveTangent=0.4051, interpMode=RCIM_Cubic, leaveTangent=0.4051, tangentMode=RCTM_User, time=0, value=0}, <at 1 1 RCIM_Cubic>]
       }}}
     CurveIndex=$Particles.Age
     Scale Curve=&RandomRangeFloat(Minimum=-0.6, Maximum=-0.25)
    )
    Rotation Vector=&Add_Vector(
     A=&CrossProduct(Vector A:Vector3f=$Particles.WallDirection, Vector B:Vector3f=<0 0 1>, Normalize Incoming Vectors:NiagaraBool=true)
     B=&RandomVector(Vector Scale=&MakeCustomFloatFromBool(
       Bool=$Particles.WoodBeams
       True Float=&RandomRangeFloat(Minimum=0.4, Maximum=1)
       False Float=&RandomRangeFloat(Minimum=0.2, Maximum=0.4)
      ))
    )
   )
   &Drag(Drag=0.25, Use Linear Drag=true, Use Rotational Drag=false)
   &GravityForce(Gravity=&Multiply_VectorByFloat(
     Vector=<0 0 -960>
     Float=&FloatFromCurve(
      FloatCurve=DataInterface() {propertyValues {Curve {
         defaultValue=1
         keys [
          {arriveTangent=6.4744, interpMode=RCIM_Cubic, leaveTangent=6.4744, tangentMode=RCTM_User, time=0, value=0.1}
          {tangentMode=RCTM_User, time=0.25, value=1}
         ]
        }}}
      Scale Curve=&RandomRangeFloat(Minimum=1, Maximum=2)
     )
    ))
   &SolveForcesAndVelocity(Clamp Velocity=false, Limit Acceleration=false)
  ]
  rendererClasses [&NiagaraMeshRendererProperties]
  renderers [
   {rendererClass=&NiagaraMeshRendererProperties, rendererIndex=0}
   {rendererClass=&NiagaraMeshRendererProperties, rendererIndex=1}
  ]
 }
}
summary {
 systemName=NS_PistolCanary_Tower_Crumble
 emitters [
  {emitterName=WallCrumbles, rendererClasses [&NiagaraMeshRendererProperties]}
  {emitterName=WoodFrames, rendererClasses [&NiagaraMeshRendererProperties]}
  {emitterName=HighMeshSmoke, rendererClasses [&NiagaraMeshRendererProperties]}
  {emitterName=RoofTiles, simTarget=GPUComputeSim, rendererClasses [&NiagaraMeshRendererProperties]}
  {emitterName=LowUnlitSmoke, rendererClasses [&NiagaraSpriteRendererProperties]}
 ]
 userVariables []
}
